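# Translucent, two-sided unlit material for Niagara flowing ribbon effects in the Slam VFX set.
# SplineThicken shapes width with an edge mask, plus an offset along the camera vector.
# Two refined trail textures on power-warped UVs build color, with tip and center brightening, while the shape texture, tip/tail fade, particle alpha and depth fade drive opacity.
# A dynamic parameter supplies fade offset, random, time override and center-light strength.
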Material /SpireSharedAssets/VFX/Slam/M_FlowRibbon
graph {
 properties {
  bAutomaticallySetUsageInEditor=false
  bUsedWithNiagaraMeshParticles=true
  bUsedWithNiagaraRibbons=true
  bUsedWithNiagaraSprites=true
  blendMode=BLEND_Translucent
  shadingModel=MSM_Unlit
  twoSided=true
  usedShadingModels=Unlit
 }
 nodes [
  n1 = TextureCoordinate()
  n2 = MaterialFunctionCall(Float2=@n1) {materialFunction=&BreakOutFloat2Components}
  n3 = VectorParameter() {group=WidthB, parameterName=WidthB_EdgeMask_InvertControl, channelNames {a="", b=Power, g=Multiply, r=Offset}, defaultValue=<rgba 0 4 1 1>}
  n4 = VectorParameter() {group=WidthB, parameterName=WidthB_EdgeMask_Control, channelNames {a="", b=Power, g=Multiply, r=Offset}, defaultValue=<rgba 5 5 1 1>}
  n5 = Multiply(
   A=ScalarParameter() {defaultValue=200, parameterName=Width}
   B=MaterialFunctionCall!(UorV=@n2.G, Invert_EdgeControl=@n3, NoInvert_EdgeControl=@n4) {materialFunction=/Game/Effects/Niagara/MasterMaterials/Functions/Common/FXCommon_EdgeMask_Ex}.Mask
  )
  n6 = MaterialFunctionCall(WidthBase=@n5, WidthTip=@n5) {materialFunction=&SplineThicken}
  n7 = Subtract(A=Add(A=WorldPosition().XYZ, B=@n6.WorldPositionOffset), B=CameraPositionWS())
  Add(A=Multiply(A=Normalize(VectorInput=@n7), B=ScalarParameter() {defaultValue=-100, parameterName=CameraOffset}), B=@n6.WorldPositionOffset)
  n8 = MaterialFunctionCall(Float2=MaterialFunctionCall(XY=@n1) {materialFunction=&Swizzle}.YX) {materialFunction=&BreakOutFloat2Components}
  n9 = VectorParameter() {
   group=TIpTailFade
   parameterName=TIpTailFade_EdgeMask_InvertControl
   channelNames {a="", b=Power, g=Multiply, r=Offset}
   defaultValue=<rgba 1 1 1 1>
  }
  n10 = VectorParameter() {
   group=TIpTailFade
   parameterName=TIpTailFade_EdgeMask_Control
   channelNames {a="", b=Power, g=Multiply, r=Offset}
   defaultValue=<rgba 0 1.2 2 1>
  }
  n11 = DynamicParameter() {defaultValue=<rgba 0 1 1 1>, paramNames [Offset,Random,Time,CenterLight]}
  n12 = Add(A=@n10.Offset, B=@n11.Offset)
  n13 = TextureObjectParameter() {
   group=RibbonShape
   parameterName=RibbonShape_Texture
   sortPriority=0
   texture=/Game/Effects/Fort_Effects/Textures/Trail/T_Pickaxe_MistMylar_Trail
  }
  n14 = MaterialFunctionCall(X=NamedRerouteUsage(), Y=Constant()) {materialFunction=&MakeFloat2}
  n15 = VectorParameter() {group=RibbonShape, parameterName=RibbonShape_UVControl, channelNames {a=VPan, b=UPan, g=VScale, r=UScale}, defaultValue=<rgba 2.5 1 0.5 0>}
  n16 = OneMinus(None=@n8.R)
  n17 = ScalarParameter() {defaultValue=0.5, parameterName=EXP}
  n18 = MaterialFunctionCall(X=OneMinus(None=Power(Base=@n16, Exp=@n17)), Y=@n8.G) {materialFunction=&MakeFloat2}
  n19 = MaterialFunctionCall(
   Texture=@n13
   AdditionalUVOffset=@n14.Result
   UVControl_RGB=@n15
   UVControl_A=@n15
   "[Override]TimeInput"=NamedRerouteUsage()
   "[Override]UVInput"=@n18.Result
  ) {materialFunction=&FXCommon_TextureSample}
  n20 = VectorParameter() {parameterName=RibbonShape_Refine, defaultValue=<rgba 6 1 4 0.5>}
  n21 = Multiply(
   A=MaterialFunctionCall(
    UorV=@n8.R
    Invert_EdgeControl=@n9
    NoInvert_EdgeControl=MaterialFunctionCall(X=@n12, Y=@n10.Multiply, Z=@n10.Multiply) {materialFunction=&MakeFloat3}.Result
   ) {materialFunction=&FXCommon_EdgeMask_Ex}.Mask
   B=MaterialFunctionCall(Texture=@n19.R, RGB=@n20.RGB, A=@n20.RGB) {materialFunction=&MF_FX_Refine}.Result
  )
  DepthFade(Opacity=Multiply(A=ParticleColor().A, B=Saturate(None=@n21)), FadeDistance=ScalarParameter() {defaultValue=50, parameterName=DF})
  n22 = VectorParameter() {
   group=TipBright
   parameterName=TipBright_EdgeMask_InvertControl
   channelNames {a="", b=Power, g=Multiply, r=Offset}
   defaultValue=<rgba 1 1 1 1>
  }
  n23 = VectorParameter() {
   group=TipBright
   parameterName=TipBright_EdgeMask_Control
   channelNames {a="", b=Power, g=Multiply, r=Offset}
   defaultValue=<rgba 0.0384 1 32.156 1>
  }
  n24 = MaterialFunctionCall(UorV=@n2.G, Invert_EdgeControl=@n22, NoInvert_EdgeControl=@n23) {materialFunction=&FXCommon_EdgeMask_Ex}
  n25 = ScalarParameter() {defaultValue=140.7303, parameterName=TipBright_C}
  n26 = VectorParameter() {
   group=Ribbon2
   parameterName=Ribbon2_EdgeMask_InvertControl
   channelNames {a="", b=Power, g=Multiply, r=Offset}
   defaultValue=<rgba 1 50 1 1>
  }
  n27 = VectorParameter() {
   group=Ribbon2
   parameterName=Ribbon2_EdgeMask_Control
   channelNames {a="", b=Power, g=Multiply, r=Offset}
   defaultValue=<rgba 0 0.898 2.5117 1>
  }
  n28 = Saturate(None=MaterialFunctionCall(UorV=@n2.G, Invert_EdgeControl=@n26, NoInvert_EdgeControl=@n27) {materialFunction=&FXCommon_EdgeMask_Ex}.Mask)
  n29 = TextureObjectParameter() {group=Ribbon2, parameterName=Ribbon2_Texture, sortPriority=0, texture=/Game/Weapons/FORT_Melee/Pickaxe_Sunlit/FX/T_Pickaxe_Sunlit_Trail_01}
  n30 = VectorParameter() {group=Ribbon2, parameterName=Ribbon2_UVControl, channelNames {a=VPan, b=UPan, g=VScale, r=UScale}, defaultValue=<rgba 2.8886 1 0.3 0>}
  n31 = VectorParameter() {parameterName=Ribbon2_Refine, defaultValue=<rgba 4.2548 2 0.5 0.5>}
  n32 = MaterialFunctionCall(
   Texture=MaterialFunctionCall(
    Texture=@n29
    AdditionalUVOffset=@n14.Result
    UVControl_RGB=@n30
    UVControl_A=@n30
    "[Override]TimeInput"=NamedRerouteUsage()
    "[Override]UVInput"=@n18.Result
   ) {materialFunction=&FXCommon_TextureSample}.R
   RGB=@n31.RGB
   A=@n31.RGB
  ) {materialFunction=&MF_FX_Refine}
  n33 = VectorParameter() {parameterName=Ribbon2Color, defaultValue=<rgba 9.13 0.21 0.9533 1>}
  n34 = Multiply(A=Add(A=Multiply(A=@n24.Mask, B=@n25)), B=Multiply(A=Multiply(A=@n28, B=@n11.CenterLight), B=Multiply(A=@n32.Result, B=@n33.RGB)))
  n35 = VectorParameter() {parameterName=RibbonColor_Refine, defaultValue=<rgba 30.8382 60 3 0.5>}
  n36 = Multiply(
   A=VectorParameter() {parameterName=C, defaultValue=<rgba 0.3 0.0189 0.0657 1>}.RGB
   B=Add(A=MaterialFunctionCall(Texture=@n19.R, RGB=@n35.RGB, A=@n35.RGB) {materialFunction=&MF_FX_Refine}.Result)
  )
  MaterialFunctionCall(Emissive=Divide(A=Add(A=@n34, B=@n36)) {constB=2.271}) {materialFunction=&MF_FX_EmissiveAdjustments}
  NamedRerouteDeclaration(None=@n11.Random) {name=Random, nodeColor=<rgba 0 1 0.9766 1>}
  NamedRerouteDeclaration(None=@n11.Time) {name=TimeIn, nodeColor=<rgba 1 0.6563 0 1>}
  TextureCoordinate()
 ]
}
