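# Translucent, unlit, two-sided material for Niagara mesh particles, drawing the HighTower Date slam ring effect.
# SplineThicken sets ring width, while blurred noise pushes vertices outward from the particle position, both driven by dynamic parameters.
# Emissive is particle color through a UV sphere mask and emissive-adjustment function.
# Opacity takes a noise-jittered curve-atlas gradient, subtracts tiled directional noise scaled by Erode, then applies particle alpha and near-camera fade.
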Material /SpireSharedAssets/VFX/Slam/M_HighTower_Date_Ring
graph {
 properties {
  bAutomaticallySetUsageInEditor=false
  bUsedWithNiagaraMeshParticles=true
  blendMode=BLEND_Translucent
  shadingModel=MSM_Unlit
  twoSided=true
  usedShadingModels=Unlit
 }
 nodes [
  n1 = Multiply(A=Constant() {r=5}, B=DynamicParameter() {defaultValue=<rgba 0 1 0 0>, paramNames [Erode,Width,WPO,IsUltra]}.Width)
  n2 = Normalize(VectorInput=Subtract(A=WorldPosition().XYZ, B=ParticlePositionWS()))
  n3 = Constant() {r=0.2}
  n4 = Multiply(A=ScalarParameter() {defaultValue=1, parameterName=U_Tile}) {constB=3}
  n5 = TextureSample(UVs=Multiply(A=TextureCoordinate(), B=AppendVector(A=@n3, B=@n4))) {samplerType=SAMPLERTYPE_Grayscale, texture=&LowResBlurredNoise}
  Add(
   A=MaterialFunctionCall(WidthBase=@n1, WidthTip=@n1) {materialFunction=&SplineThicken}.WorldPositionOffset
   B=Multiply(A=Multiply(A=@n2, B=@n5.R), B=DynamicParameter() {defaultValue=<rgba 0 1 0 0>, paramNames [Erode,Width,WPO,IsUltra]}.WPO)
  )
  Constant2Vector() {g=1, r=0.2}
  Constant2Vector() {g=3, r=0.2}
  n6 = ParticleColor()
  n7 = Multiply(A=TextureCoordinate(), B=Constant2Vector() {g=1, r=0.18})
  n8 = Constant2Vector() {g=0.5, r=-1.5}
  n9 = ScalarParameter() {defaultValue=1.57, parameterName=EmissiveSphereMaskRadius}
  MaterialFunctionCall(Emissive=LinearInterpolate(B=@n6.RGB, Alpha=SphereMask(A=@n7, B=@n8, Radius=@n9) {attenuationRadius=1.7, hardnessPercent=95})) {materialFunction=&MF_FX_EmissiveAdjustments}
  n10 = TextureCoordinate()
  n11 = Constant2Vector() {g=2, r=0.2}
  n12 = LinearInterpolate(Alpha=TextureSample(UVs=Multiply(A=@n10, B=@n11)) {samplerType=SAMPLERTYPE_LinearColor, texture=/Game/Global/Textures/Default/Noise/T_Noise01_Inverse}.R) {constA=-1, constB=5}
  n13 = ScalarParameter() {defaultValue=12}
  n14 = CurveAtlasRowParameter(CurveTime=Add(
    A=MaterialFunctionCall() {materialFunction=&LinearGradient}.UGradient
    B=Divide(A=@n12, B=@n13) {constB=30}
   )) {
   atlas=/Game/Effects/Fort_Effects/GameplayPulginsAssets/Demon/CA_HighTower_Date
   curve=/Game/Effects/Fort_Effects/GameplayPulginsAssets/Demon/Curve_HighTower_Date_Ball
   parameterName=Param_1
  }
  n15 = LinearInterpolate(Alpha=DynamicParameter() {defaultValue=<rgba 0 1 0 0>, paramNames [Erode,Width,WPO,IsUltra]}.Erode) {constA=1, constB=10}
  n16 = Constant() {r=0.2}
  n17 = Multiply(
   A=ScalarParameter() {defaultValue=1, parameterName=U_Tile}
   B=DynamicParameter() {parameterIndex=1, defaultValue=<rgba 0 1 0 0>, paramNames [U_Tile,Param2,Param3,Param4]}.U_Tile
  )
  n18 = TextureSample(UVs=Multiply(A=TextureCoordinate(), B=AppendVector(A=@n16, B=@n17))) {samplerType=SAMPLERTYPE_LinearGrayscale, texture=/Game/Effects/Fort_Effects/Emotes/UltraLight/Texture/T_DirectionalNoise_Greyscale}
  n19 = Power(Base=@n18.R)
  n20 = LinearInterpolate(Alpha=@n18.R) {constA=0.8}
  n21 = MaterialFunctionCall() {materialFunction=&LinearGradient}
  n22 = Multiply(A=OneMinus(None=LinearInterpolate(A=@n19, B=@n20, Alpha=@n21.UGradient)))
  n23 = Saturate(None=Subtract(A=Multiply(A=@n14.G), B=Multiply(A=@n15, B=@n22)))
  n24 = Constant() {r=2}
  n25 = ScalarParameter() {defaultValue=100, parameterName=Camera Fade distance}
  Multiply(
   A=Multiply(A=@n6.A, B=Multiply(A=@n23, B=@n24) {constB=2})
   B=Power(Base=MaterialFunctionCall(Distance=@n25) {materialFunction=&Near_Camera_Fade}.Result)
  )
  ScalarParameter() {defaultValue=1, parameterName=U_Tile}
  VectorParameter() {parameterName=IsNotUltraColor, defaultValue=<rgba 0.4206 1 0.1009 1>}
 ]
}
