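# Translucent, unlit, two-sided Niagara material for a holographic ring in the Slam VFX, apparently a telegraph.
# A noise-distorted main texture, tiled on one axis and clamped on the other, drives a two-color emissive blend tinted by particle color.
# Opacity adds a second ramp texture, then applies squared depth fade, optional inverse Fresnel and particle alpha.
# Dynamic parameters supply animation time and scale SplineThicken width.
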Material /SpireSharedAssets/VFX/Slam/M_HoloTellRing
graph {
 properties {
  bAutomaticallySetUsageInEditor=false
  bUsedWithNiagaraMeshParticles=true
  bUsedWithNiagaraRibbons=true
  bUsedWithNiagaraSprites=true
  blendMode=BLEND_Translucent
  shadingModel=MSM_Unlit
  twoSided=true
  usedShadingModels=Unlit
 }
 nodes [
  n1 = DynamicParameter() {paramNames [W,Param2,Param3,Time]}
  n2 = Multiply(A=ScalarParameter() {parameterName=SplineWidth}, B=@n1.W)
  MaterialFunctionCall(WidthBase=@n2, WidthTip=@n2) {materialFunction=&SplineThicken}
  n3 = TextureObjectParameter() {
   group=MainTexture
   parameterName=MainTexture_Texture
   sortPriority=0
   texture=/Game/Effects/Fort_Effects/Textures/Trail/T_Pickaxe_MistMylar_Trail
  }
  n4 = TextureObjectParameter() {
   group=Noise
   parameterName=Noise_Texture
   samplerType=SAMPLERTYPE_Normal
   sortPriority=0
   texture=/Game/Characters/Player/Female/Medium/Bodies/F_MED_Bistro_Astronaut/FX/T_GaussianNoise_Normal
  }
  n5 = VectorParameter() {group=Noise, parameterName=Noise_UVControl, channelNames {a=VPan, b=UPan, g=VScale, r=UScale}, defaultValue=<rgba 1 1 0 0.3>}
  n6 = TextureCoordinate()
  n7 = StaticSwitchParameter(True=MaterialFunctionCall(XY=@n6) {materialFunction=&Swizzle}.YX, False=@n6) {parameterName=SwitchUV}
  n8 = VectorParameter() {
   group=Noise
   parameterName=Noise_NoiseControl
   sortPriority=60
   channelNames {a=VStrength, b=UStrength, g=VOffset, r=UOffset}
   defaultValue=<rgba -0.5 -0.5 0 0.1>
  }
  n9 = VectorParameter() {group=MainTexture, parameterName=MainTexture_UVControl, channelNames {a=VPan, b=UPan, g=VScale, r=UScale}, defaultValue=<rgba 2 2.5 0.2 0>}
  n10 = MaterialFunctionCall(XY=@n7) {materialFunction=&Swizzle}
  n11 = MaterialFunctionCall!(
   Texture=@n3
   AdditionalUVOffset=MaterialFunctionCall!(
    NoiseValue=MaterialFunctionCall(Texture=@n4, UVControl_RGB=@n5, UVControl_A=@n5, "[Override]UVInput"=@n7) {materialFunction=&FXCommon_TextureSample}.RG
    Control_RGB=@n8
    Control_A=@n8
   ) {materialFunction=/Game/Effects/Niagara/MasterMaterials/Functions/Common/FXCommon_NoiseControl}.NoiseValue
   UVControl_RGB=@n9
   UVControl_A=@n9
   "[Override]TimeInput"=NamedRerouteUsage()
   "[Override]ScaleUVByCenter"=StaticBool() {value=true}
   "[Override]UVInput"=@n10.YX
  ) {materialFunction=/Game/Effects/Niagara/MasterMaterials/Functions/Common/FXCommon_TextureSample_V2}
  n12 = Multiply(
   A=MaterialFunctionCall!(TexCoord=@n11.UV, ClampChannel=Constant2Vector() {g=1}) {materialFunction=/Game/Effects/Fort_Effects/MaterialFunctions/MF_FX_TilingMethod}.Result
   B=@n11.R
  )
  n13 = VectorParameter() {parameterName=RingTexture_RefineHighLight, defaultValue=<rgba 10 1 1 0.5>}
  n14 = LinearInterpolate(
   A=VectorParameter() {parameterName=C1, defaultValue=<rgba 1 0 0.1333 1>}.RGB
   B=VectorParameter() {parameterName=C2, defaultValue=<rgba 0.5 0.025 0.0646 1>}.RGB
   Alpha=MaterialFunctionCall(Texture=@n12, RGB=@n13.RGB, A=@n13.RGB) {materialFunction=&MF_FX_Refine}.Result
  )
  n15 = ParticleColor()
  MaterialFunctionCall(Emissive=Divide(A=Multiply(A=@n14, B=@n15.RGB)) {constB=2.271}) {materialFunction=&MF_FX_EmissiveAdjustments}
  n16 = VectorParameter() {group=Noise, parameterName=Ribbon2_UVControl, channelNames {a=VPan, b=UPan, g=VScale, r=UScale}, defaultValue=<rgba 1 2.4118 0 0>}
  n17 = MaterialFunctionCall(
   Texture=TextureObjectParameter() {group=Noise, parameterName=Ribbon2_Texture, sortPriority=0, texture=&/T_Water_Top_Ramp}
   UVControl_RGB=@n16
   UVControl_A=@n16
   "[Override]ScaleUVByCenter"=StaticBool() {value=true}
   "[Override]UVInput"=@n10.YX
  ) {materialFunction=&FXCommon_TextureSample_V2}
  n18 = MaterialFunctionCall(TexCoord=@n17.UV, ClampChannel=Constant2Vector() {g=1}) {materialFunction=&MF_FX_TilingMethod}
  n19 = VectorParameter() {parameterName=Ribbon2_Refine, defaultValue=<rgba 0.5 0.2 0 0>}
  n20 = VectorParameter() {parameterName=RingTexture2_RefineHighLight, defaultValue=<rgba 0 0 0.7 1>}
  n21 = Add(
   A=MaterialFunctionCall(Texture=Multiply(A=@n17.R, B=@n18.Result), RGB=@n19.RGB, A=@n19.RGB) {materialFunction=&MF_FX_Refine}.Result
   B=MaterialFunctionCall(Texture=@n12, RGB=@n20.RGB, A=@n20.RGB) {materialFunction=&MF_FX_Refine}.Result
  )
  n22 = DepthFade(FadeDistance=ScalarParameter() {defaultValue=40, parameterName=DF})
  n23 = Multiply(A=@n22, B=@n22)
  n24 = ScalarParameter() {defaultValue=4, parameterName=EXP}
  n25 = Power(Base=OneMinus(None=Fresnel(ExponentIn=@n24) {baseReflectFraction=0}), Exp=ScalarParameter() {defaultValue=2, parameterName=EXP2})
  n26 = ScalarParameter() {parameterName=DisableFres}
  Multiply(A=Multiply(A=Multiply(A=@n21, B=@n23), B=LinearInterpolate(A=@n25, Alpha=@n26)), B=@n15.A)
  NamedRerouteDeclaration(None=@n1.Time) {name=TimeIn, nodeColor=<rgba 0.6875 1 0 1>}
 ]
}
