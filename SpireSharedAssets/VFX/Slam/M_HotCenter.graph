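# Unlit, two-sided translucent Niagara particle material for the Slam VFX, drawing a soft radial glow with a hot center.
# A UV-centered sphere mask, shaped by nested power and one-minus curves, blends BaseColor into HotColor.
# That color is tinted by particle color, divided down and passed through MF_FX_EmissiveAdjustments.
# A second, wider mask with its own curve parameters multiplies particle alpha to form the circular falloff.
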Material /SpireSharedAssets/VFX/Slam/M_HotCenter
graph {
 properties {
  bAutomaticallySetUsageInEditor=false
  bUsedWithNiagaraMeshParticles=true
  bUsedWithNiagaraRibbons=true
  bUsedWithNiagaraSprites=true
  blendMode=BLEND_Translucent
  shadingModel=MSM_Unlit
  twoSided=true
  usedShadingModels=Unlit
 }
 nodes [
  n1 = TextureCoordinate()
  n2 = Constant() {r=0.5}
  n3 = ScalarParameter() {defaultValue=0.3296, parameterName=CenterHotRadius}
  n4 = ScalarParameter() {parameterName=CH_Hardness}
  n5 = OneMinus(None=Power(Base=SphereMask(A=@n1, B=@n2, Radius=@n3, Hardness=@n4) {hardnessPercent=0}, Exp=ScalarParameter() {defaultValue=2, parameterName=CH_A}))
  n6 = ScalarParameter() {defaultValue=1.2257, parameterName=CH_B}
  n7 = LinearInterpolate(
   A=VectorParameter() {parameterName=BaseColor, defaultValue=<rgba 0.1 0 0.0083 1>}.RGB
   B=VectorParameter() {parameterName=HotColor, defaultValue=<rgba 4 0.2 0.5167 1>}.RGB
   Alpha=OneMinus(None=Power(Base=@n5, Exp=@n6))
  )
  n8 = ParticleColor()
  MaterialFunctionCall(Emissive=Divide(A=Multiply(A=@n7, B=@n8.RGB)) {constB=2.271}) {materialFunction=&MF_FX_EmissiveAdjustments}
  n9 = Constant() {r=0.5}
  n10 = ScalarParameter() {defaultValue=0.0276, parameterName=Hd}
  n11 = OneMinus(None=Power(Base=SphereMask(A=@n1, B=@n2, Radius=@n9, Hardness=@n10) {hardnessPercent=0}, Exp=ScalarParameter() {defaultValue=1.5, parameterName=A}))
  n12 = ScalarParameter() {defaultValue=4.5547, parameterName=B}
  Multiply(A=@n8.A, B=OneMinus(None=Power(Base=@n11, Exp=@n12)))
 ]
}
