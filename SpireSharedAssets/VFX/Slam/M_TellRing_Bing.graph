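# Additive unlit, two-sided Niagara material for a slam VFX tell ring, with USEWBTDecal switching between decal-projected and mesh radial UVs.
# Two edge masks on one UV axis, one shifted by Offset, are tinted by C2 and C, summed and multiplied by particle color for emissive.
# A tiled, Frac-scrolled sweep mask driven by the dynamic Time parameter and Speed multiplies the refined ring mask and particle alpha.
# Dynamic parameters also supply glow strength and decal bound size.
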Material /SpireSharedAssets/VFX/Slam/M_TellRing_Bing
graph {
 properties {
  bAutomaticallySetUsageInEditor=false
  bUsedWithNiagaraMeshParticles=true
  bUsedWithNiagaraRibbons=true
  bUsedWithNiagaraSprites=true
  blendMode=BLEND_Additive
  shadingModel=MSM_Unlit
  twoSided=true
  usedShadingModels=Unlit
 }
 nodes [
  n1 = NamedRerouteUsage()
  n2 = MaterialFunctionCall(Float2=@n1) {materialFunction=&BreakOutFloat2Components}
  n3 = Add(
   A=@n2.G
   B=MaterialFunctionCall(X=ScalarParameter() {defaultValue=-0.4099, parameterName=Offset}, Y=Constant()) {materialFunction=&MakeFloat2}.Result
  )
  n4 = VectorParameter() {
   group=ShapeB
   parameterName=ShapeB_EdgeMask_InvertControl
   channelNames {a="", b=Power, g=Multiply, r=Offset}
   defaultValue=<rgba 0 2 2708.835 1>
  }
  n5 = VectorParameter() {group=ShapeB, parameterName=ShapeB_EdgeMask_Control, channelNames {a="", b=Power, g=Multiply, r=Offset}, defaultValue=<rgba 0 2 103.2033 1>}
  n6 = DynamicParameter() {paramNames [SoftGlowSTR,Time,Param3,Param4]}
  n7 = VectorParameter() {parameterName=C, defaultValue=<rgba 10 0 0.8333 1>}
  n8 = VectorParameter() {group=ShapeA, parameterName=ShapeA_EdgeMask_InvertControl, channelNames {a="", b=Power, g=Multiply, r=Offset}, defaultValue=<rgba 0 10 2 1>}
  n9 = VectorParameter() {group=ShapeA, parameterName=ShapeA_EdgeMask_Control, channelNames {a="", b=Power, g=Multiply, r=Offset}, defaultValue=<rgba 0 1 2 1>}
  n10 = MaterialFunctionCall!(UorV=@n2.G, Invert_EdgeControl=@n8, NoInvert_EdgeControl=@n9) {materialFunction=/Game/Effects/Niagara/MasterMaterials/Functions/Common/FXCommon_EdgeMask_Ex}
  n11 = Add(
   A=Multiply(
    A=MaterialFunctionCall(UorV=@n3, Invert_EdgeControl=@n4, NoInvert_EdgeControl=@n5) {materialFunction=&FXCommon_EdgeMask_Ex}.Mask
    B=VectorParameter() {parameterName=C2, defaultValue=<rgba 20 1 2.5833 1>}.RGB
   )
   B=Multiply(A=@n6.SoftGlowSTR, B=Multiply(A=@n7.RGB, B=@n10.Mask))
  )
  n12 = ParticleColor()
  MaterialFunctionCall(Emissive=Divide(A=Multiply(A=@n11, B=@n12.RGB)) {constB=2.271}) {materialFunction=&MF_FX_EmissiveAdjustments}
  n13 = OneMinus(None=Multiply(A=@n1, B=Constant2Vector() {g=1, r=3}))
  n14 = Multiply(A=@n6.Time, B=ScalarParameter() {defaultValue=2, parameterName=Speed})
  n15 = MaterialFunctionCall(Float2=Frac(None=Add(A=@n13, B=@n14))) {materialFunction=&BreakOutFloat2Components}
  n16 = VectorParameter() {group=RotateS, parameterName=RotateS_EdgeMask_InvertControl, channelNames {a="", b=Power, g=Multiply, r=Offset}, defaultValue=<rgba 0 2 1 1>}
  n17 = VectorParameter() {group=RotateS, parameterName=RotateS_EdgeMask_Control, channelNames {a="", b=Power, g=Multiply, r=Offset}, defaultValue=<rgba 0 2 1 1>}
  n18 = VectorParameter() {parameterName=RotateS_Refine, defaultValue=<rgba 66.5735 27.9234 2 0.5>}
  n19 = VectorParameter() {parameterName=ShapeA_Refine, defaultValue=<rgba 10 5 0.3 0.5>}
  Multiply(
   A=MaterialFunctionCall(
    Texture=MaterialFunctionCall(UorV=@n15.R, Invert_EdgeControl=@n16, NoInvert_EdgeControl=@n17) {materialFunction=&FXCommon_EdgeMask_Ex}.Mask
    RGB=@n18.RGB
    A=@n18.RGB
   ) {materialFunction=&MF_FX_Refine}.Result
   B=Multiply(
    A=@n12.A
    B=MaterialFunctionCall(Texture=@n10.Mask, RGB=@n19.RGB, A=@n19.RGB) {materialFunction=&MF_FX_Refine}.Result
   )
  )
  n20 = DynamicParameter() {parameterIndex=3, defaultValue=<rgba 0 1000 1 1>, paramNames [BoundX,BoundY,BoundZ,Param4]}
  n21 = MaterialFunctionCall!(MeshOriginalBoundSize=MaterialFunctionCall(X=@n20.BoundX, Y=@n20.BoundX, Z=@n20.BoundX) {materialFunction=&MakeFloat3}.Result) {materialFunction=/Game/Effects/Niagara/MasterMaterials/Functions/Common/FXCommon_WBTDecalUV}
  NamedRerouteDeclaration(None=StaticSwitchParameter(
    True=@n21.RadialUV
    False=MaterialFunctionCall() {materialFunction=&VectorToRadialValue}.Radial Coordinates
   ) {defaultValue=true, parameterName=USEWBTDecal}) {name=RadialUVOut, nodeColor=<rgba 0 1 0.2031 1>}
  NamedRerouteDeclaration(None=StaticSwitchParameter(True=@n21.UV, False=TextureCoordinate()) {defaultValue=true, parameterName=USEWBTDecal}) {name=UVOut, nodeColor=<rgba 1 0 0.1172 1>}
  NamedRerouteDeclaration(None=@n21."CustomDepthMask(1onMobile)") {name=CustomDepthMask, nodeColor=<rgba 1 0 0.2812 1>}
  NamedRerouteDeclaration(None=@n21.ZMask) {name=ZMask, nodeColor=<rgba 0 0.2031 1 1>}
  Time() {bOverride_Period=true, period=6000}
 ]
}
