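# Translucent unlit Niagara material for a glowing light element of the slam tell ring VFX.
# Particle color times a tint parameter is scaled down by a constant and passed through a shared emissive adjustment function.
# Opacity combines particle and tint alpha with an edge-fade mask along the V coordinate and an inverted, powered Fresnel.
# The Fresnel term fades grazing angles, softening the silhouette.
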Material /SpireSharedAssets/VFX/Slam/M_TellRing_LightC
graph {
 properties {
  bAutomaticallySetUsageInEditor=false
  bUsedWithNiagaraMeshParticles=true
  bUsedWithNiagaraRibbons=true
  bUsedWithNiagaraSprites=true
  blendMode=BLEND_Translucent
  shadingModel=MSM_Unlit
  usedShadingModels=Unlit
 }
 nodes [
  n1 = ParticleColor()
  n2 = VectorParameter() {parameterName=C, defaultValue=<rgba 2 0 0.1667 0.7>}
  MaterialFunctionCall(Emissive=Divide(A=Multiply(A=@n1.RGB, B=@n2.RGB)) {constB=2.271}) {materialFunction=&MF_FX_EmissiveAdjustments}
  n3 = VectorParameter() {
   group=EdgeFade
   parameterName=EdgeFade_EdgeMask_InvertControl
   channelNames {a="", b=Power, g=Multiply, r=Offset}
   defaultValue=<rgba 0 50 1 1>
  }
  n4 = VectorParameter() {group=EdgeFade, parameterName=EdgeFade_EdgeMask_Control, channelNames {a="", b=Power, g=Multiply, r=Offset}, defaultValue=<rgba 0 1 1 1>}
  n5 = MaterialFunctionCall!(
   UorV=MaterialFunctionCall(Float2=TextureCoordinate()) {materialFunction=&BreakOutFloat2Components}.G
   Invert_EdgeControl=@n3
   NoInvert_EdgeControl=@n4
  ) {materialFunction=/Game/Effects/Niagara/MasterMaterials/Functions/Common/FXCommon_EdgeMask_Ex}
  n6 = ScalarParameter() {defaultValue=0.7, parameterName=EXP}
  n7 = Power(Base=OneMinus(None=Fresnel(ExponentIn=@n6) {baseReflectFraction=0}), Exp=ScalarParameter() {defaultValue=3, parameterName=EXP1})
  Multiply(A=@n1.A, B=Multiply(A=@n2.A, B=Multiply(A=@n5.Mask, B=@n7)))
 ]
}
