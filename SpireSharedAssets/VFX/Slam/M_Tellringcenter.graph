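# Unlit translucent two-sided Niagara material for the center of a slam VFX tell ring.
# Particle color tints the emissive color parameter; particle alpha drives opacity with a refined, panning plasma texture and an edge mask.
# A WBT decal UV function, sized by a dynamic bound parameter, supplies UVs, radial UVs and depth/Z masks through named reroutes.
# The USEWBTDecal static switch, on by default, falls back to mesh UVs and computed radial coordinates.
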Material /SpireSharedAssets/VFX/Slam/M_Tellringcenter
graph {
 properties {
  bAutomaticallySetUsageInEditor=false
  bUsedWithNiagaraMeshParticles=true
  bUsedWithNiagaraRibbons=true
  bUsedWithNiagaraSprites=true
  blendMode=BLEND_Translucent
  shadingModel=MSM_Unlit
  twoSided=true
  usedShadingModels=Unlit
 }
 nodes [
  n1 = ParticleColor()
  n2 = VectorParameter() {parameterName=C, defaultValue=<rgba 0.9 0.009 0.1278 1>}
  MaterialFunctionCall(Emissive=Divide(A=Multiply(A=@n1.RGB, B=@n2.RGB)) {constB=2.271}) {materialFunction=&MF_FX_EmissiveAdjustments}
  n3 = TextureObjectParameter() {
   group=MainTexture
   parameterName=MainTexture_Texture
   samplerType=SAMPLERTYPE_LinearGrayscale
   sortPriority=0
   texture=/Game/Effects/Fort_Effects/Emotes/UltraLight/Texture/T_Plasma_B_Greyscale
  }
  n4 = VectorParameter() {
   group=MainTexture
   parameterName=MainTexture_UVControl
   channelNames {a=VPan, b=UPan, g=VScale, r=UScale}
   defaultValue=<rgba 0 0.25 100 -0.5>
  }
  n5 = NamedRerouteUsage()
  n6 = VectorParameter() {parameterName=MainTexture_Refine, defaultValue=<rgba 0.7 2 1 0.5>}
  n7 = MaterialFunctionCall(
   Texture=MaterialFunctionCall(Texture=@n3, UVControl_RGB=@n4, UVControl_A=@n4, "[Override]UVInput"=@n5) {materialFunction=&FXCommon_TextureSample}.R
   RGB=@n6.RGB
   A=@n6.RGB
  ) {materialFunction=&MF_FX_Refine}
  n8 = VectorParameter() {
   group=MainTexture
   parameterName=MainTexture_EdgeMask_InvertControl
   channelNames {a="", b=Power, g=Multiply, r=Offset}
   defaultValue=<rgba 0 5 2 1>
  }
  n9 = VectorParameter() {
   group=MainTexture
   parameterName=MainTexture_EdgeMask_Control
   channelNames {a="", b=Power, g=Multiply, r=Offset}
   defaultValue=<rgba 0 3 2 1>
  }
  n10 = MaterialFunctionCall!(
   UorV=MaterialFunctionCall(Float2=@n5) {materialFunction=&BreakOutFloat2Components}.G
   Invert_EdgeControl=@n8
   NoInvert_EdgeControl=@n9
  ) {materialFunction=/Game/Effects/Niagara/MasterMaterials/Functions/Common/FXCommon_EdgeMask_Ex}
  Multiply(A=Multiply(A=@n1.A, B=Multiply(A=@n7.Result, B=@n10.Mask)), B=NamedRerouteUsage())
  n11 = DynamicParameter() {parameterIndex=3, defaultValue=<rgba 0 1000 1 1>, paramNames [BoundX,BoundY,BoundZ,Param4]}
  n12 = MaterialFunctionCall!(MeshOriginalBoundSize=MaterialFunctionCall(X=@n11.BoundX, Y=@n11.BoundX, Z=@n11.BoundX) {materialFunction=&MakeFloat3}.Result) {materialFunction=/Game/Effects/Niagara/MasterMaterials/Functions/Common/FXCommon_WBTDecalUV}
  NamedRerouteDeclaration(None=StaticSwitchParameter(
    True=@n12.RadialUV
    False=MaterialFunctionCall() {materialFunction=&VectorToRadialValue}.Radial Coordinates
   ) {defaultValue=true, parameterName=USEWBTDecal}) {name=RadialUVOut, nodeColor=<rgba 0 1 0.2031 1>}
  NamedRerouteDeclaration(None=StaticSwitchParameter(True=@n12.UV, False=TextureCoordinate()) {defaultValue=true, parameterName=USEWBTDecal}) {name=UVOut, nodeColor=<rgba 1 0 0.1172 1>}
  NamedRerouteDeclaration(None=@n12."CustomDepthMask(1onMobile)") {name=CustomDepthMask, nodeColor=<rgba 1 0 0.2812 1>}
  NamedRerouteDeclaration(None=@n12.ZMask) {name=ZMask, nodeColor=<rgba 0 0.2031 1 1>}
 ]
}
