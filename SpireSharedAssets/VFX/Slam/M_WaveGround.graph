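# Translucent, unlit, two-sided Niagara material for the slam VFX ground wave, distorting the scene behind it.
# An edge mask on the U coordinate and particle red blend refraction strength from neutral toward Ref.
# That strength and a SplineThicken normal drive a pixel-normal offset into SceneColor, with particle alpha as opacity.
# Platform switches zero color and alpha on Switch, Android and PC-Perf, disabling the effect there.
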Material /SpireSharedAssets/VFX/Slam/M_WaveGround
graph {
 properties {
  bAutomaticallySetUsageInEditor=false
  bUsedWithNiagaraMeshParticles=true
  bUsedWithNiagaraRibbons=true
  bUsedWithNiagaraSprites=true
  blendMode=BLEND_Translucent
  shadingModel=MSM_Unlit
  twoSided=true
  usedShadingModels=Unlit
 }
 nodes [
  n1 = TextureCoordinate()
  n2 = VectorParameter() {group=Mask, parameterName=Mask_EdgeMask_Control, channelNames {a="", b=Power, g=Multiply, r=Offset}, defaultValue=<rgba 0 2 2 1>}
  n3 = LinearInterpolate(
   B=ScalarParameter() {defaultValue=-0.5792, parameterName=Ref}
   Alpha=MaterialFunctionCall!(
    UorV=MaterialFunctionCall(Float2=@n1) {materialFunction=&BreakOutFloat2Components}.R
    EdgeControl=@n2
   ) {materialFunction=/Game/Effects/Niagara/MasterMaterials/Functions/Common/FXCommon_EdgeMask_Simple}.Mask
  ) {constA=1}
  n4 = ParticleColor()
  n5 = ScalarParameter() {defaultValue=500, parameterName=Width}
  n6 = SceneColor!(OffsetFraction=MaterialFunctionCall!(
    RefractionStrength=LinearInterpolate(B=@n3, Alpha=@n4.R) {constA=1}
    DepthBias=ScalarParameter() {defaultValue=100, parameterName=DepthBias}
    FalloffDistance=ScalarParameter() {defaultValue=400, parameterName="FalloffDistance (S)"}
    NormalTS=MaterialFunctionCall(WidthBase=@n5, WidthTip=@n5) {materialFunction=&SplineThicken}.Normal
   ) {materialFunction=/Game/Effects/Niagara/MasterMaterials/Functions/Common/FXCommon_PixelNormalOffsetRefraction}.ScreenUVOffsetFraction) {inputMode=OffsetFraction}
  n7 = Constant()
  MaterialFunctionCall!("Xb1,PS4"=@n6.RGB, "PC-High&&Epic,XSX,PS5"=@n6.RGB, "PC-Perf,Android"=@n7, Switch=@n7, "PC-Low(NotPerf)&&Medium"=@n6.RGB) {materialFunction=/Game/Effects/Niagara/MasterMaterials/Functions/Common/PlatformSwitch_3D}
  MaterialFunctionCall("Xb1,PS4"=@n4.A, "PC-High&&Epic,XSX,PS5"=@n4.A, "PC-Perf,Android"=@n7, Switch=@n7, "PC-Low(NotPerf)&&Medium"=@n4.A) {materialFunction=&PlatformSwitch_3D}
 ]
}
