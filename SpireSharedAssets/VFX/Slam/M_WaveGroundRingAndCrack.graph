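# Translucent unlit Niagara material for a slam VFX, drawing a glowing shockwave ring with cracks on the ground.
# A decal UV function supplies projected and radial UVs, which two noise layers distort, with dynamic parameters driving ring offset and noise strength.
# Two edge masks shape the ring and a center band broken up by a refined texture, then tinted for emissive.
# Their combined alpha is multiplied by particle alpha, custom-depth and Z masks, with depth testing disabled.
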Material /SpireSharedAssets/VFX/Slam/M_WaveGroundRingAndCrack
graph {
 properties {
  bAutomaticallySetUsageInEditor=false
  bDisableDepthTest=true
  bUsedWithNiagaraMeshParticles=true
  bUsedWithNiagaraRibbons=true
  bUsedWithNiagaraSprites=true
  blendMode=BLEND_Translucent
  shadingModel=MSM_Unlit
  usedShadingModels=Unlit
 }
 nodes [
  n1 = TextureObjectParameter() {
   group=Noise
   parameterName=Noise_Texture
   samplerType=SAMPLERTYPE_LinearColor
   sortPriority=0
   texture=/Game/Effects/Fort_Effects/BluGlow/Violentium/Violentium_Mask
  }
  n2 = VectorParameter() {group=Noise, parameterName=Noise_UVControl, channelNames {a=VPan, b=UPan, g=VScale, r=UScale}, defaultValue=<rgba 4 4 0 0>}
  n3 = MaterialFunctionCall!(MeshOriginalBoundSize) {materialFunction=/Game/Effects/Niagara/MasterMaterials/Functions/Common/FXCommon_WBTDecalUV}
  n4 = MaterialFunctionCall(Texture=@n1, UVControl_RGB=@n2, UVControl_A=@n2, "[Override]UVInput"=@n3.UV) {materialFunction=&FXCommon_TextureSample}
  n5 = VectorParameter() {
   group=Noise
   parameterName=Noise_NoiseControl
   sortPriority=60
   channelNames {a=VStrength, b=UStrength, g=VOffset, r=UOffset}
   defaultValue=<rgba -0.5 0 0 0.2>
  }
  n6 = Multiply(A=@n5.VStrength, B=DynamicParameter() {defaultValue=<rgba 0 1 1 1>, paramNames [Offset,NoiseA,Param3,Param4]}.NoiseA)
  n7 = Add(
   A=@n3.RadialUV
   B=Add(
    A=DynamicParameter() {defaultValue=<rgba 0 1 1 1>, paramNames [Offset,NoiseA,Param3,Param4]}.Offset
    B=ScalarParameter() {defaultValue=0.3408, parameterName=Offe}
   )
  )
  n8 = TextureObjectParameter() {
   group=Noise2
   parameterName=Noise2_Texture
   samplerType=SAMPLERTYPE_Normal
   sortPriority=0
   texture=/Game/Characters/Player/Female/Medium/Bodies/F_MED_Bistro_Astronaut/FX/T_GaussianNoise_Normal
  }
  n9 = VectorParameter() {group=Noise2, parameterName=Noise2_UVControl, channelNames {a=VPan, b=UPan, g=VScale, r=UScale}, defaultValue=<rgba 7 1 0 0.2>}
  n10 = VectorParameter() {
   group=Noise2
   parameterName=Noise2_NoiseControl
   sortPriority=60
   channelNames {a=VStrength, b=UStrength, g=VOffset, r=UOffset}
   defaultValue=<rgba -0.5 -0.5 0 0.1888>
  }
  n11 = MaterialFunctionCall!(
   NoiseValue=MaterialFunctionCall(Texture=@n8, UVControl_RGB=@n9, UVControl_A=@n9, "[Override]UVInput"=@n3.RadialUV) {materialFunction=&FXCommon_TextureSample}.RG
   Control_RGB=@n10
   Control_A=@n10
  ) {materialFunction=/Game/Effects/Niagara/MasterMaterials/Functions/Common/FXCommon_NoiseControl}
  n12 = MaterialFunctionCall(Float2=Add(
    A=MaterialFunctionCall(NoiseValue=@n4.RG, Control_RGB=@n5, Control_A=@n6) {materialFunction=&FXCommon_NoiseControl}.NoiseValue
    B=Add(A=@n7, B=@n11.NoiseValue)
   )) {materialFunction=&BreakOutFloat2Components}
  n13 = VectorParameter() {
   group=Ring
   parameterName=Ring_EdgeMask_InvertControl
   channelNames {a="", b=Power, g=Multiply, r=Offset}
   defaultValue=<rgba 0 7.8579 2.2181 1>
  }
  n14 = VectorParameter() {group=Ring, parameterName=Ring_EdgeMask_Control, channelNames {a="", b=Power, g=Multiply, r=Offset}, defaultValue=<rgba 0 1.1497 6.2831 1>}
  n15 = MaterialFunctionCall!(UorV=@n12.G, Invert_EdgeControl=@n13, NoInvert_EdgeControl=@n14) {materialFunction=/Game/Effects/Niagara/MasterMaterials/Functions/Common/FXCommon_EdgeMask_Ex}
  n16 = Multiply(A=VectorParameter() {defaultValue=<rgba 2 0.1 0.2583 1>}.RGB, B=@n15.Mask)
  n17 = VectorParameter() {
   group=RingCent
   parameterName=RingCent_EdgeMask_InvertControl
   channelNames {a="", b=Power, g=Multiply, r=Offset}
   defaultValue=<rgba -0.1088 20 1.8639 1>
  }
  n18 = VectorParameter() {
   group=RingCent
   parameterName=RingCent_EdgeMask_Control
   channelNames {a="", b=Power, g=Multiply, r=Offset}
   defaultValue=<rgba 0 1.135 85147.5312 1>
  }
  n19 = TextureObjectParameter() {
   group=BTexture
   parameterName=BTexture_Texture
   samplerType=SAMPLERTYPE_LinearColor
   sortPriority=0
   texture=/Engine/Functions/Engine_MaterialFunctions02/ExampleContent/Textures/OffsetNoiseDistanceFields_ForNormals
  }
  n20 = VectorParameter() {group=BTexture, parameterName=BTexture_UVControl, channelNames {a=VPan, b=UPan, g=VScale, r=UScale}, defaultValue=<rgba 1 1 0 0>}
  n21 = VectorParameter() {parameterName=BTexture_Refine, defaultValue=<rgba 3 5 0 0.5>}
  n22 = Multiply(
   A=MaterialFunctionCall(UorV=@n12.G, Invert_EdgeControl=@n17, NoInvert_EdgeControl=@n18) {materialFunction=&FXCommon_EdgeMask_Ex}.Mask
   B=MaterialFunctionCall(
    Texture=MaterialFunctionCall(Texture=@n19, UVControl_RGB=@n20, UVControl_A=@n20, "[Override]UVInput"=@n3.UV) {materialFunction=&FXCommon_TextureSample}.R
    RGB=@n21.RGB
    A=@n21.RGB
   ) {materialFunction=&MF_FX_Refine}.Result
  )
  n23 = Multiply(A=VectorParameter() {parameterName=Param2, defaultValue=<rgba 500 50 87.499 1>}.RGB, B=@n22)
  MaterialFunctionCall(Emissive=Divide(A=Add(A=@n16, B=@n23)) {constB=2.271}) {materialFunction=&MF_FX_EmissiveAdjustments}
  MaterialFunctionCall() {materialFunction=&VectorToRadialValue}
  n24 = ParticleColor()
  n25 = ScalarParameter() {defaultValue=0.7, parameterName=A}
  n26 = ScalarParameter() {defaultValue=100, parameterName=B}
  n27 = Saturate(None=Add(A=Multiply(A=@n15.Mask, B=@n25), B=Multiply(A=@n26, B=@n22)))
  Multiply(A=Multiply(A=@n3."CustomDepthMask(1onMobile)", B=Multiply(A=@n24.A, B=@n27)), B=@n3.ZMask)
 ]
}
