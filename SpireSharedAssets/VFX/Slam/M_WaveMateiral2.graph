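# Translucent, unlit, two-sided wave material for the Slam VFX, enabled for Niagara sprites, ribbons and mesh particles.
# Color blends a solid tint with scene color multiplied by a second tint, then multiplies by particle color; performance PC, Android and Switch skip the scene color multiply.
# Opacity combines a U-axis edge mask, particle alpha, a scalar and squared depth fade.
# SplineThicken takes one Width parameter for base and tip.
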Material /SpireSharedAssets/VFX/Slam/M_WaveMateiral2
graph {
 properties {
  bAutomaticallySetUsageInEditor=false
  bUsedWithNiagaraMeshParticles=true
  bUsedWithNiagaraRibbons=true
  bUsedWithNiagaraSprites=true
  blendMode=BLEND_Translucent
  shadingModel=MSM_Unlit
  twoSided=true
  usedShadingModels=Unlit
 }
 nodes [
  n1 = ScalarParameter() {defaultValue=413.5033, parameterName=Width}
  MaterialFunctionCall(WidthBase=@n1, WidthTip=@n1) {materialFunction=&SplineThicken}
  n2 = Divide(A=VectorParameter() {parameterName=Color, defaultValue=<rgba 1 0.1 0.175 1>}.RGB) {constB=2.271}
  n3 = MaterialFunctionCall(Emissive=Divide(A=VectorParameter() {parameterName=ScaleSC, defaultValue=<rgba 5 0.565 0.9346 1>}.RGB) {constB=2.271}) {materialFunction=&MF_FX_EmissiveAdjustments}
  n4 = Multiply(A=SceneColor!(UVs).RGB, B=@n3.Result)
  n5 = ParticleColor()
  Multiply(
   A=LinearInterpolate(
    A=MaterialFunctionCall(Emissive=@n2) {materialFunction=&MF_FX_EmissiveAdjustments}.Result
    B=MaterialFunctionCall!("Xb1,PS4"=@n4, "PC-High&&Epic,XSX,PS5"=@n4, "PC-Perf,Android"=@n3.Result, Switch=@n3.Result, "PC-Low(NotPerf)&&Medium"=@n4) {materialFunction=/Game/Effects/Niagara/MasterMaterials/Functions/Common/PlatformSwitch_3D}.Result
    Alpha=ScalarParameter() {defaultValue=0.3194, parameterName=BSC}
   )
   B=MaterialFunctionCall(Emissive=Divide(A=@n5.RGB) {constB=2.271}) {materialFunction=&MF_FX_EmissiveAdjustments}.Result
  )
  n6 = VectorParameter() {group=Shape, parameterName=Shape_EdgeMask_InvertControl, channelNames {a="", b=Power, g=Multiply, r=Offset}, defaultValue=<rgba 0 2 2 1>}
  n7 = VectorParameter() {group=Shape, parameterName=Shape_EdgeMask_Control, channelNames {a="", b=Power, g=Multiply, r=Offset}, defaultValue=<rgba 0 2 2 1>}
  n8 = MaterialFunctionCall!(
   UorV=MaterialFunctionCall(Float2=TextureCoordinate()) {materialFunction=&BreakOutFloat2Components}.R
   Invert_EdgeControl=@n6
   NoInvert_EdgeControl=@n7
  ) {materialFunction=/Game/Effects/Niagara/MasterMaterials/Functions/Common/FXCommon_EdgeMask_Ex}
  n9 = ScalarParameter() {defaultValue=1, parameterName=O}
  n10 = DepthFade(FadeDistance=ScalarParameter() {defaultValue=50, parameterName=DF})
  Multiply(A=Multiply(A=Multiply(A=@n8.Mask, B=@n9), B=@n5.A), B=Multiply(A=@n10, B=@n10))
 ]
}
