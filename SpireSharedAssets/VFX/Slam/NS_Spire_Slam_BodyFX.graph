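# Niagara body VFX for the Spire slam: a red-pink energy charge that surges around 1.7 seconds.
# A flame mesh with a flickering light combines with rising smoke, curl-noise sparks and three streaks stretched toward a shared point.
# The Spline emitter orients and stretches a mesh toward a sampled skeletal-mesh position.
# Curves indexed by system age synchronise the build-up, flash and fade; glow and flare emitters show no content.
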NiagaraSystem /SpireSharedAssets/VFX/Slam/NS_Spire_Slam_BodyFX
dependencies {
 usedDataInterfaces [
  &NiagaraDataInterfaceCurve
  &NiagaraDataInterfaceColorCurve
  &NiagaraDataInterfaceArrayFloat
  &NiagaraDataInterfaceSkeletalMesh
 ]
 usedDynamicInputs []
 usedModules [
  &SystemState
  &self:SystemSpawnScript.SystemScriptSource.SystemScriptGraph.NiagaraNodeAssignment_0.SetVariables_#1
  &self:SystemSpawnScript.SystemScriptSource.SystemScriptGraph.NiagaraNodeAssignment_1.SetVariables_#2
  &SampleSkeletalMesh
 ]
}
emitters {
 Flame_Source_Strum {
  emitterUpdateScript [
   &EmitterState(
    Inactive Response=::NewEnumerator1
    Life Cycle Mode=::NewEnumerator1
    Loop Behavior=::NewEnumerator1
    Loop Delay=0.2
    Reset Age On Awaken=true
    Scalability Mode=::NewEnumerator1
    UseLoopDelay=true
   )
   &SpawnBurst_Instantaneous(Spawn Count=1, Use Spawn Probability=false)
   /SetVariables(
    Emitter.Depth2:NiagaraFloat=&FloatFromCurve(FloatCurve=DataInterface() {propertyValues {Curve {keys [<at 0 30>,<at 1 30>,<at 3.1 30>,<at 3.5 100>]}}}, CurveIndex=$Engine.System.Age)
    Emitter.DepthFade:NiagaraFloat=&FloatFromCurve(FloatCurve=DataInterface() {propertyValues {Curve {keys [<at 0 30>,<at 1 30>,<at 3.1 30>,<at 3.5 300>]}}}, CurveIndex=$Engine.System.Age)
   ) {isSetParameters=true}
  ]
  particleSpawnScript [
   &PartitionParticles(Number Of Partitions=1)
   &InitializeParticle(
    Color=<rgba 1 1 1 0.5>
    Lifetime=5
    Mesh Scale=<10 10 28.6301>
    Mesh Scale Mode=::NewEnumerator4
    Position Offset=<20 0 -317.3212>
    UsePositionOffset=true
    Write Material Random=true
   ) {module=InitializeParticle001}
   /SetVariables(Particles.Mesh_Size2:Vector3f=<3 3 12>) {isSetParameters=true}
   &CameraOffset(Camera Offset Amount=&FloatFromCurve(FloatCurve=DataInterface() {propertyValues {Curve {keys [<at 3 1>,<at 3.5 1>]}}}, CurveIndex=$Engine.System.Age, Scale Curve=100))
   &InitialMeshOrientation(Mesh Orientation Mode=::NewEnumerator2, Use Rotation Vector=true)
  ]
  particleUpdateScript [
   /SetVariables(Particles.Mesh_Color2:LinearColor=&MakeLinearColorFromVectorAndFloat(
     "Vector (RGB)"=&MakeVectorFromLinearColorRGB(Color=&MultiplyLinearColor(Linear Color=<rgba 3 0.4844 0.9036 1>, Scale Factor=&MakeVector4_FromFloat(Float=1)))
     "Float (Alpha)"=&FloatFromCurve(
      FloatCurve=DataInterface() {propertyValues {Curve {keys [{interpMode=RCIM_Cubic, leaveTangent=0.0545, tangentMode=RCTM_Break, time=0.4, value=0}, <at 1.6 0.1>, <at 1.7 1>]}}}
      CurveIndex=$Engine.System.Age
      Scale Curve=&FloatFromCurve(
       FloatCurve=DataInterface() {propertyValues {Curve {keys [{interpMode=RCIM_Cubic, leaveTangent=-0.0232, tangentMode=RCTM_Break, time=0.4, value=0}, <at 1.6 0.5>, <at 1.7 1>, <at 4 1>, <at 4.3 0>]}}}
       CurveIndex=$Engine.System.Age
       Scale Curve=0.5
      )
     )
    )) {isSetParameters=true}
   &ParticleState()
   &Timeline(
    Play Rate=&FloatFromCurve(FloatCurve=DataInterface() {propertyValues {Curve {keys [<at 1.6 1>,<at 1.7 2>]}}}, CurveIndex=$Engine.System.Age)
    Timeline End=9999
   )
   /SetVariables(Particles.ScaleSize:NiagaraFloat=&FloatFromCurve(
     FloatCurve=DataInterface() {propertyValues {Curve {keys [
         <at 0 0.2>
         <at 0.2 0.5>
         {interpMode=RCIM_Cubic, leaveTangent=3.2761, tangentMode=RCTM_Break, time=1.55, value=0.7}
         {arriveTangent=1.0704, interpMode=RCIM_Cubic, leaveTangent=1.0704, tangentMode=RCTM_User, time=1.8, value=1.4}
         <at 2.5 1.3 RCIM_Cubic>
        ]}}}
     CurveIndex=$Engine.System.Age
     Scale Curve=&FloatFromCurve(
      FloatCurve=DataInterface() {propertyValues {Curve {keys [
          <at 0 1>
          <at 1 1>
          {interpMode=RCIM_Cubic, tangentMode=RCTM_Break, time=3.1, value=1}
          {interpMode=RCIM_Cubic, tangentMode=RCTM_Break, time=3.4, value=1.5}
          <at 4 1.1>
         ]}}}
      CurveIndex=$Engine.System.Age
     )
    )) {isSetParameters=true}
   /SetVariables(Particles.Mesh_Size2:Vector3f=&Multiply_VectorByFloat(
     Vector=<3 3 12>
     Float=&Multiply_Float(
      A=$Particles.ScaleSize
      B=&FloatFromCurve(
       FloatCurve=DataInterface() {propertyValues {Curve {keys [
           <at 0 1>
           <at 1 1>
           {interpMode=RCIM_Cubic, tangentMode=RCTM_Break, time=3.1, value=1}
           {interpMode=RCIM_Cubic, tangentMode=RCTM_Break, time=3.4, value=1.7}
           <at 4 1.3>
          ]}}}
       CurveIndex=$Engine.System.Age
      )
     )
    )) {isSetParameters=true}
   &ScaleMeshSize(Scale Factor=&VectorFromFloat(Value=$Particles.ScaleSize))
   &SolveForcesAndVelocity(Clamp Velocity=false, Limit Acceleration=false)
   &FN_Light_Attributes(
    Color=<rgba 1 0.2035 0.2699 1>
    Create Color Attribute=true
    Create Position Attribute=true
    Create Radius Attribute=true
    Enable Particle Light Enable=true
    Light Enabled=true
    Position=<100 0 -178.046>
    Radius=5000
    Scale Color=&Waveform(
     Global Amplitude Scale=&Multiply_Float(
      A=230139.2344
      B=&FloatFromCurve(
       FloatCurve=DataInterface() {propertyValues {Curve {keys [
           <at 0.5 0>
           <at 1.6 0.4>
           {arriveTangent=0.4286, interpMode=RCIM_Cubic, leaveTangent=0.4286, time=1.7, value=1}
           {arriveTangent=-0.7143, interpMode=RCIM_Cubic, leaveTangent=-0.7143, time=3, value=1}
           <at 3.1 0 RCIM_Cubic>
          ]}}}
       CurveIndex=$Engine.System.Age
      )
     )
     "Waveform [1]"=::NewEnumerator12
     "[1] Frequency"=5
     "[1] Amplitude Min/Max"=<0.7 1>
    )
   )
   /SetVariables(Particles.Mesh2_Dynamic:Vector4f=&MakeVector4(
     X=&SelectFloatFromArray(Direct Array Index=0, Float Selection Array=DataInterface() {propertyValues {FloatData [0.35,0.8]}})
     Y=1
     Z=30
     W=0
    )) {isSetParameters=true}
   &DynamicMaterialParameters(
    Index 0 Param 1=&RandomRangeFloat(Minimum=0, Maximum=0)
    Index 0 Param 2=1
    Index 3 Param 1=$Particles.Timeline.Time
    Module.Param0WriteEnabled=true
    Module.Param0WriteEnabled3=true
    Module.Param1WriteEnabled=true
    Module.Param1WriteEnabled3=true
    Module.Param2WriteEnabled=true
    Module.Param2WriteEnabled3=true
    Module.Param3WriteEnabled=true
    Module.Param3WriteEnabled3=true
    Write Parameter Index 3=true
   )
   &ScaleColor(
    Scale Alpha=&FloatFromCurve(
     FloatCurve=DataInterface() {propertyValues {Curve {keys [{interpMode=RCIM_Cubic, leaveTangent=-0.0232, tangentMode=RCTM_Break, time=0.4, value=0}, <at 1.6 0.5>, <at 1.7 1>, <at 4 1>, <at 4.3 0>]}}}
     CurveIndex=$Engine.System.Age
    )
    ScaleA=true
    ScaleRGB=true
   )
  ]
  rendererClasses [&NiagaraMeshRendererProperties, &NiagaraLightRendererProperties]
  renderers [
   {rendererClass=&NiagaraMeshRendererProperties, rendererIndex=0}
   {rendererClass=&NiagaraMeshRendererProperties, rendererIndex=1}
   {rendererClass=&NiagaraLightRendererProperties, rendererIndex=2}
  ]
 }
 Flare001 {rendererClasses [], renderers []}
 Glow1 {rendererClasses [], renderers []}
 Glow2 {rendererClasses [], renderers []}
 Glow3 {rendererClasses [], renderers []}
 GlowDecal {rendererClasses [], renderers []}
 IdleSmoke001 {
  emitterUpdateScript [
   &EmitterState(Life Cycle Mode=::NewEnumerator1, Loop Behavior=::NewEnumerator1, Loop Delay=0.5, Loop Duration=4, UseLoopDelay=true)
   /SetVariables(Emitter.LightingCenter:Vector4f=&MakeFloat4FromVectorAndFloat(
     "Vector (RGB)"=&TransformPosition(Position=<0 0 100>, Source Space=::Local, Destination Space=::World)
     "Float (Alpha)"=1
    )) {isSetParameters=true}
   &SpawnRate(SpawnRate=5, Use Spawn Probability=false)
  ]
  particleSpawnScript [
   &InitializeParticle(
    Color=<rgba 1 0.8783 0.9373 1>
    Lifetime=&RandomRangeFloat(Minimum=0.9, Maximum=1.1)
    Position Offset=<0 0 -100>
    Sprite Rotation Mode=::NewEnumerator2
    Sprite Size Mode=::NewEnumerator3
    Uniform Sprite Size=&RandomRangeFloat(Minimum=450, Maximum=500)
    UsePositionOffset=true
    Write Material Random=true
   ) {module=InitializeParticle001}
   &CameraOffset(Camera Offset Amount=&RandomRangeFloat(Minimum=0, Maximum=10))
   &ShapeLocation(
    Cylinder Distribution=::NewEnumerator2
    Cylinder Height=200
    Cylinder Radius=1
    MaskSpawnGroup=false
    Shape Primitive=::NewEnumerator1
    UseOwnerScale=true
    V Position=&ModuloFloat(A=&Multiply_Float(A=$Emitter.Age, B=5), B=1)
   )
   /SetVariables(Particles.SpriteSize=&Multiply_Vector2D_ByFloat(
     Float=&FloatFromCurve(
      FloatCurve=DataInterface() {propertyValues {Curve {keys [{interpMode=RCIM_Cubic, leaveTangent=0.4547, tangentMode=RCTM_Break, time=0, value=0.6}, <at 1 1>]}}}
      CurveIndex=&ModuloFloat(A=&Multiply_Float(A=$Emitter.Age, B=5), B=1)
     )
     Vector2D=$Particles.SpriteSize
    )) {isSetParameters=true}
   &AddVelocity(Velocity=<0 0 300>)
  ]
  particleUpdateScript [
   &ParticleState()
   &ScaleSpriteSize(Scale Factor=&Vector2DFromFloat(Value=&FloatFromCurve(FloatCurve=DataInterface() {propertyValues {Curve {keys [
          {interpMode=RCIM_Cubic, leaveTangent=0.6892, tangentMode=RCTM_Break, time=0, value=0.5}
          {arriveTangent=0.1753, tangentMode=RCTM_User, time=0.8, value=0.9}
         ]}}})))
   /SetVariables(Particles.Color_ParticleEnergy:LinearColor=&ColorFromCurve(ColorCurve=DataInterface() {propertyValues {
       AlphaCurve {keys [<at 0 1 RCIM_Cubic>]}
       BlueCurve {keys [<at 0.3364 0.3332>,<at 0.749 0.0641 RCIM_Cubic>]}
       GreenCurve {keys [<at 0.3364 0.3531>,<at 0.749 0 RCIM_Cubic>]}
       RedCurve {keys [<at 0.3364 2>,<at 0.749 5 RCIM_Cubic>]}
      }})) {isSetParameters=true}
   &InheritSourceMovement(Applied Position Scale=<0.8 0.8 0.8>, ApplyForce=false, ApplyPosition=true, ApplyPositionFalloff=false) {enabled=false}
   &Drag(Drag=0, Use Linear Drag=true, Use Rotational Drag=true)
   &SolveForcesAndVelocity(Clamp Velocity=false, Limit Acceleration=false)
   &DynamicMaterialParameters(
    Index 0 Param 1=&FloatFromCurve(FloatCurve=DataInterface() {propertyValues {Curve {keys [{tangentMode=RCTM_User, time=0, value=0}, {tangentMode=RCTM_User, time=1, value=1}]}}})
    Index 0 Param 3=&FloatFromCurve(
     FloatCurve=DataInterface() {propertyValues {Curve {keys [
         {interpMode=RCIM_Cubic, tangentMode=RCTM_Break, time=0, value=0}
         {interpMode=RCIM_Cubic, leaveTangent=-3.8707, tangentMode=RCTM_Break, time=0.2, value=1}
         {tangentMode=RCTM_User, time=1, value=0}
        ]}}}
     Scale Curve=2
    )
    Index 0 Param 4=1
    Index 1 Param 1=&MakeFloatFromVector(Vector=&MakeVectorFromLinearColorRGB(Color=$Particles.Color_ParticleEnergy), Channel=::NewEnumerator0)
    Index 1 Param 2=&MakeFloatFromVector(Vector=&MakeVectorFromLinearColorRGB(Color=$Particles.Color_ParticleEnergy), Channel=::NewEnumerator1)
    Index 1 Param 3=&MakeFloatFromVector(Vector=&MakeVectorFromLinearColorRGB(Color=$Particles.Color_ParticleEnergy), Channel=::NewEnumerator2)
    Index 1 Param 4=&FloatFromCurve(FloatCurve=DataInterface() {propertyValues {Curve {keys [
         {interpMode=RCIM_Cubic, leaveTangent=0.736, tangentMode=RCTM_Break, time=0.4, value=0.2}
         {arriveTangent=0.1446, tangentMode=RCTM_User, time=1, value=1}
        ]}}})
    Index 2 Param 2=600
    Index 2 Param 3=300
    Index 2 Param 4=-0.25
    Module.Index0Float0Write=true
    Module.Index0Float1Write=true
    Module.Index0Float2Write=true
    Module.Index0Float3Write=true
    Module.Index1Float0Write=true
    Module.Index1Float1Write=true
    Module.Index1Float2Write=true
    Module.Index1Float3Write=true
    Module.Index2Float0Write=true
    Module.Index2Float1Write=true
    Module.Index2Float2Write=true
    Module.Index2Float3Write=true
    Write Parameter Index 1=::NewEnumerator1
    Write Parameter Index 2=::NewEnumerator1
   )
   &ScaleColor(
    Scale Alpha=&FloatFromCurve(FloatCurve=DataInterface() {propertyValues {Curve {keys [
         {interpMode=RCIM_Cubic, leaveTangent=7.3008, tangentMode=RCTM_Break, time=0, value=0}
         {interpMode=RCIM_Cubic, tangentMode=RCTM_Break, time=0.3, value=1}
         {interpMode=RCIM_Cubic, leaveTangent=-0.0057, tangentMode=RCTM_Break, time=0.4, value=1}
         {arriveTangent=-4.3242, interpMode=RCIM_Cubic, leaveTangent=-4.3242, tangentMode=RCTM_User, time=1, value=0}
        ]}}})
    ScaleA=true
    ScaleRGB=true
   )
   &ScaleColor(
    Scale Alpha=&FloatFromCurve(
     FloatCurve=DataInterface() {propertyValues {Curve {keys [
         {interpMode=RCIM_Cubic, tangentMode=RCTM_Break, time=0.5, value=0}
         {interpMode=RCIM_Cubic, tangentMode=RCTM_Break, time=1.55, value=0.1}
         <at 1.8 1>
        ]}}}
     CurveIndex=$Engine.System.Age
    )
    ScaleA=true
    ScaleRGB=true
   ) {module=ScaleColor001}
  ]
  rendererClasses [&NiagaraSpriteRendererProperties]
  renderers [{rendererClass=&NiagaraSpriteRendererProperties}]
 }
 Particle {
  emitterUpdateScript [
   &EmitterState(
    Inactive Response=::NewEnumerator1
    Life Cycle Mode=::NewEnumerator1
    Loop Behavior=::NewEnumerator1
    Loop Delay=1.7
    Loop Duration=4
    UseLoopDelay=true
   )
   &SpawnBurst_Instantaneous(Spawn Count=10, Use Spawn Probability=false, UseLoopCountLimit=false)
   &SpawnRate(SpawnRate=10, Use Spawn Probability=false)
  ]
  particleSpawnScript [
   &InitializeParticle(
    Color=<rgba 1.3 0.65 0.7042 1>
    Lifetime=&RandomRangeFloat(Minimum=0.7, Maximum=1)
    Sprite Size Max=<1.5 4>
    Sprite Size Min=<0.8 3>
    Sprite Size Mode=::NewEnumerator2
    UsePositionOffset=false
    Write Material Random=true
   )
   &ShapeLocation(MaskSpawnGroup=false, Offset=<0 0 -205.8731>, Offset Mode=::NewEnumerator0, Override Seed=false, Sphere Radius=40, UseOwnerScale=true)
   &AddVelocity(
    Cone Angle=60
    Cone Axis=<0 0 1>
    Module.UseConeFalloff=false
    Module.UseVelDistribution=false
    Override Seed=false
    Velocity Mode=::NewEnumerator2
    Velocity Speed=&RandomRangeFloat(Minimum=676.9317, Maximum=1283.0527)
   )
   &CameraOffset(Camera Offset Amount=100)
  ]
  particleUpdateScript [
   &ParticleState()
   &ScaleSpriteSize(Scale Sprite Size Mode=::NewEnumerator0, Uniform Scale Factor=3)
   &CurlNoiseForce(
    Bias Noise=false
    Noise Frequency=20
    Noise Strength=&FloatFromCurve(
     FloatCurve=DataInterface() {propertyValues {Curve {keys [
         {arriveTangent=-1.6035, interpMode=RCIM_Cubic, leaveTangent=-1.6035, tangentMode=RCTM_User, time=0, value=1}
         {arriveTangent=-1.46, interpMode=RCIM_Cubic, leaveTangent=-1.46, tangentMode=RCTM_User, time=0.4618, value=0.3676}
         <at 1 0>
        ]}}}
     Scale Curve=400
    )
    Pan Noise=false
   )
   &Drag(
    Drag=&FloatFromCurve(FloatCurve=DataInterface() {propertyValues {Curve {keys [<at 0 1>,<at 1 0>]}}})
    Use Linear Drag=true
    Use Rotational Drag=true
   )
   &SolveForcesAndVelocity(Clamp Velocity=false, Limit Acceleration=false)
   &ScaleSpriteSizeBySpeed(Max Scale Factor=<1 3>, Min Scale Factor=<0.5 0.5>)
   &FN_Sprite_Attributes(
    Color=&MultiplyLinearColor(
     Linear Color=<rgba 1 0.1 0.175 0.2>
     Scale Factor=&MakeVector4(
      X=1
      Y=1
      Z=1
      W=&FloatFromCurve(FloatCurve=DataInterface() {propertyValues {Curve {keys [{interpMode=RCIM_Cubic, tangentMode=RCTM_Break, time=0.5, value=1}, <at 1 0>]}}})
     )
    )
    Create Color Attribute=true
    Create SpriteSize Attribute=true
    SpriteSize=&Multiply_Vector2D(A=$Particles.SpriteSize, B=<20 4>)
    Write to Module Namespace=true
   )
   &DynamicMaterialParameters(
    Index 0 Param 1=2
    Module.Index0Param0Write=true
    Module.Index0Param1Write=true
    Module.Index0Param2Write=true
    Module.Index0Param3Write=true
   )
   &ScaleColor(
    Scale Alpha=&FloatFromCurve(FloatCurve=DataInterface() {propertyValues {Curve {keys [
         {interpMode=RCIM_Cubic, tangentMode=RCTM_User, time=0, value=0}
         {interpMode=RCIM_Cubic, tangentMode=RCTM_User, time=0.5, value=1}
         {tangentMode=RCTM_User, time=1, value=0}
        ]}}})
    ScaleA=true
    ScaleRGB=true
   )
  ]
  rendererClasses [&NiagaraSpriteRendererProperties]
  renderers [
   {rendererClass=&NiagaraSpriteRendererProperties}
   {rendererClass=&NiagaraSpriteRendererProperties, rendererIndex=1}
  ]
 }
 RedLine {
  emitterUpdateScript [
   &EmitterState(Life Cycle Mode=::NewEnumerator1, Loop Behavior=::NewEnumerator1, Loop Delay=1, Loop Duration=4, UseLoopDelay=true)
   &SpawnBurst_Instantaneous(Spawn Count=3, Use Spawn Probability=false, UseLoopCountLimit=false)
  ]
  particleSpawnScript [
   &FN_SetSpawnGroup(Number of Spawn Groups=3)
   &InitializeParticle(
    Color=<rgba 1 0 0.0833 5>
    Lifetime=2
    Position Offset=&Multiply_VectorByFloat(
     Vector=&MergeHelper_VectorSwitch(0=<150 0 600>, 1=<-500 250 900>, 2=<-500 -250 900>, 3=?, Index=$Particles.SpawnGroup, Num=::NewEnumerator1)
     Float=1.3
    )
    Sprite Size=<120 1>
    Sprite Size Mode=::NewEnumerator4
    UsePositionOffset=true
    Write Material Random=true
   )
   &CameraOffset(Camera Offset Amount=70)
  ]
  particleUpdateScript [
   &ParticleState()
   /SetVariables(Particles.TargetPosition:NiagaraPosition=&Multiply_VectorByFloat(Vector=<20 0 -280>, Float=1)) {isSetParameters=true}
   &ScaleSpriteSize(
    Scale Factor=&MakeVector2D(
     X=1
     Y=&Multiply_Float(
      A=&Multiply_Float(
       A=&DistanceBetweenPositions(A=$Particles.TargetPosition, B=$Particles.Initial.Position)
       B=&MakeFloatFromVector(Vector=$Engine.Owner.Scale, Channel=::NewEnumerator0)
      )
      B=&FloatFromCurve(
       FloatCurve=DataInterface() {propertyValues {Curve {keys [
           {interpMode=RCIM_Cubic, tangentMode=RCTM_Break, time=0.5, value=0}
           {interpMode=RCIM_Cubic, tangentMode=RCTM_Break, time=1.2, value=0}
           {arriveTangent=3.2942, interpMode=RCIM_Cubic, leaveTangent=3.2942, tangentMode=RCTM_User, time=1.6, value=0.5}
           <at 1.8 1>
          ]}}}
       CurveIndex=$Engine.System.Age
      )
     )
    )
    Scale Sprite Size Mode=::NewEnumerator2
   )
   &SpriteFacingAndAlignment(
    Sprite Alignment=&SubtractVectorFromPosition(Position=$Particles.Initial.Position, Vector=Linked() {linkedVariable {name=Particles.TargetPosition, type=NiagaraPosition}})
    WriteAlignment=true
    WriteFacing=false
   )
   &ScaleColor(
    Scale Alpha=&FloatFromCurve(FloatCurve=DataInterface() {propertyValues {Curve {keys [
         {interpMode=RCIM_Cubic, tangentMode=RCTM_Break, time=0, value=0}
         {interpMode=RCIM_Cubic, tangentMode=RCTM_Break, time=0.1, value=1}
         {arriveTangent=-0.0065, interpMode=RCIM_Cubic, leaveTangent=-0.0065, tangentMode=RCTM_User, time=0.7, value=1}
         <at 0.9 0>
        ]}}})
    Scale RGB=&VectorFromFloat(Value=&FloatFromCurve(
      FloatCurve=DataInterface() {propertyValues {Curve {keys [
          {interpMode=RCIM_Cubic, tangentMode=RCTM_Break, time=0.5, value=1}
          {interpMode=RCIM_Cubic, tangentMode=RCTM_Break, time=1.5, value=1}
          {interpMode=RCIM_Cubic, tangentMode=RCTM_Break, time=1.8, value=3}
          <at 2 1>
         ]}}}
      CurveIndex=$Engine.System.Age
     ))
    ScaleA=true
    ScaleRGB=true
   ) {module=ScaleColor001}
   &DynamicMaterialParameters(
    Index 0 Param 1=&RandomRangeFloat(Minimum=0, Maximum=5)
    Module.Index0Float0Write=true
    Module.Index0Float1Write=true
    Module.Index0Float2Write=true
    Module.Index0Float3Write=true
   )
  ]
  rendererClasses [&NiagaraSpriteRendererProperties]
  renderers [{rendererClass=&NiagaraSpriteRendererProperties}]
 }
 Spline {
  emitterUpdateScript [
   &EmitterState(Inactive Response=::NewEnumerator1, Life Cycle Mode=::NewEnumerator1, Loop Behavior=::NewEnumerator1, Loop Duration=4, UseLoopDelay=false)
   &SpawnBurst_Instantaneous(Use Spawn Probability=false, UseLoopCountLimit=false)
   &SpawnBurst_Instantaneous(Spawn Group=1, Use Spawn Probability=false, UseLoopCountLimit=false) {module=SpawnBurst_Instantaneous001}
  ]
  particleSpawnScript [
   &InitializeParticle(
    Color=&MaskLinearColorBySpawnGroup(Masked Linear Color=<rgba 1 0.03 0.1917 0.2>, Passthrough Linear Color=<rgba 1 0 0.1667 1>)
    Lifetime=&MaskFloatBySpawnGroup(Masked Float=4, Passthrough Float=3)
    Mesh Scale Mode=::NewEnumerator3
    Sprite Size=<1300 350>
    Sprite Size Mode=::NewEnumerator4
    UsePositionOffset=false
    Write Material Random=true
   )
   &CameraOffset(Camera Offset Amount=&MaskFloatBySpawnGroup(Masked Float=10, Passthrough Float=500))
   /SetVariables(Particles.VisibilityTag=&SpawnGroup()) {isSetParameters=true}
  ]
  particleUpdateScript [
   &ParticleState()
   /SetVariables(Particles.Position=&MaskVectorBySpawnGroup(
     Masked Vector=Linked() {linkedVariable {name=Engine.Emitter.SimulationPosition, type=NiagaraPosition}}
     Passthrough Vector=Linked() {linkedVariable {name=System.SampleSkeletalMesh.SampledPosition, type=NiagaraPosition}}
    )) {isSetParameters=true}
   &ScaleMeshSize(Scale Factor=&MakeVector(
     X=1
     Y=1
     Z=&Multiply_Float(
      A=&Divide_Float(
       A=&VectorLength(Vector=&Subtract_Vector(
         A=Linked() {linkedVariable {name=Engine.Emitter.SimulationPosition, type=NiagaraPosition}}
         B=Linked() {linkedVariable {name=System.SampleSkeletalMesh.SampledPosition, type=NiagaraPosition}}
        ))
       B=&Multiply_Float(A=10, B=1)
      )
      B=&FloatFromCurve(FloatCurve=DataInterface() {propertyValues {Curve {keys [<at 0 1>, {arriveTangent=0.1406, time=0.2, value=1}]}}})
     )
    ))
   /SetVariables(Particles.MeshOrientation=&FindQuatBetweenTwoVectors(
     From Vector=<0 0 1>
     To Vector=&NormalizeVector(
      Vector To Normalize=&Subtract_Vector(
       A=Linked() {linkedVariable {name=System.SampleSkeletalMesh.SampledPosition, type=NiagaraPosition}}
       B=Linked() {linkedVariable {name=Engine.Emitter.SimulationPosition, type=NiagaraPosition}}
      )
      Safe Normalize=true
      Safe Fallback Threshold=0
      Safe Fallback Vector=<0 0 0.001>
     )
    )) {isSetParameters=true}
   &ScaleColor(
    Scale Alpha=&Divide_Float(
     A=1183.0546
     B=&VectorLength(Vector=&Subtract_Vector(
       A=Linked() {linkedVariable {name=Engine.Emitter.SimulationPosition, type=NiagaraPosition}}
       B=Linked() {linkedVariable {name=System.SampleSkeletalMesh.SampledPosition, type=NiagaraPosition}}
      ))
    )
    ScaleA=true
    ScaleRGB=true
   ) {enabled=false}
   &Timeline(Play Rate=-1.5, Rewind Mode:ENiagara_TimelineMode=::NewEnumerator1, Timeline Beginning=-999999, Timeline End=9999999)
   &DynamicMaterialParameters(
    Index 0 Param 1=&MaskFloatBySpawnGroup(Masked Float=$Particles.Timeline.Time, Passthrough Float=1)
    Index 0 Param 2=&MaskFloatBySpawnGroup(
     Masked Float=&FloatFromCurve(FloatCurve=DataInterface() {propertyValues {Curve {keys [
          {interpMode=RCIM_Cubic, leaveTangent=7.954, tangentMode=RCTM_Break, time=0, value=0}
          {interpMode=RCIM_Cubic, tangentMode=RCTM_Break, time=0.3, value=1}
          <at 1 1>
         ]}}})
     Passthrough Float=1
    )
    Module.Param0WriteEnabled=true
    Module.Param1WriteEnabled=true
    Module.Param2WriteEnabled=true
    Module.Param3WriteEnabled=true
   )
   &ScaleColor(
    Scale Alpha=&MaskFloatBySpawnGroup(
     Masked Float=&FloatFromCurve(
      FloatCurve=DataInterface() {propertyValues {Curve {keys [
          {interpMode=RCIM_Cubic, tangentMode=RCTM_Break, time=0, value=0}
          {arriveTangent=0.9798, interpMode=RCIM_Cubic, leaveTangent=-3.4063, tangentMode=RCTM_Break, time=0.3, value=1}
          {arriveTangent=-3.0955, tangentMode=RCTM_User, time=1, value=0}
         ]}}}
      Scale Curve=&Divide_Float(
       A=1183.0546
       B=&VectorLength(Vector=&Subtract_Vector(
         A=Linked() {linkedVariable {name=Engine.Emitter.SimulationPosition, type=NiagaraPosition}}
         B=Linked() {linkedVariable {name=System.SampleSkeletalMesh.SampledPosition, type=NiagaraPosition}}
        ))
      )
     )
     Passthrough Float=&FloatFromCurve(FloatCurve=DataInterface() {propertyValues {Curve {
         defaultValue=0
         keys [
          {arriveTangent=9.0725, interpMode=RCIM_Cubic, leaveTangent=9.0725, tangentMode=RCTM_User, time=0, value=0}
          {arriveTangent=-2.5249, interpMode=RCIM_Cubic, leaveTangent=-2.5249, tangentMode=RCTM_User, time=0.2, value=0.9847}
          {arriveTangent=-0.0405, interpMode=RCIM_Cubic, leaveTangent=-0.0405, tangentMode=RCTM_User, time=1, value=0}
         ]
        }}})
    )
    Scale RGB=&MaskVectorBySpawnGroup(
     Masked Vector=&VectorFromFloat(Value=&FloatFromCurve(
       FloatCurve=DataInterface() {propertyValues {Curve {keys [
           {interpMode=RCIM_Cubic, tangentMode=RCTM_Break, time=0, value=0.7}
           {interpMode=RCIM_Cubic, tangentMode=RCTM_Break, time=1.7, value=1}
           {interpMode=RCIM_Cubic, leaveTangent=-5.0261, tangentMode=RCTM_Break, time=1.8, value=3}
           <at 2.5 1>
          ]}}}
       CurveIndex=$Engine.System.Age
      ))
     Passthrough Vector=<1 1 1>
    )
    ScaleA=true
    ScaleRGB=true
   ) {module=ScaleColor001}
  ]
  rendererClasses [&NiagaraMeshRendererProperties, &NiagaraSpriteRendererProperties]
  renderers [
   {rendererClass=&NiagaraMeshRendererProperties, rendererIndex=0}
   {rendererClass=&NiagaraSpriteRendererProperties, rendererIndex=1}
  ]
 }
}
summary {
 systemName=NS_Spire_Slam_BodyFX
 emitters [
  {emitterName=Glow3, rendererClasses []}
  {emitterName=GlowDecal, rendererClasses []}
  {emitterName=IdleSmoke001, rendererClasses [&NiagaraSpriteRendererProperties]}
  {
   emitterName=Flame_Source_Strum
   rendererClasses [&NiagaraMeshRendererProperties, &NiagaraLightRendererProperties]
  }
  {emitterName=Glow1, rendererClasses []}
  {emitterName=Glow2, rendererClasses []}
  {emitterName=Particle, rendererClasses [&NiagaraSpriteRendererProperties]}
  {emitterName=RedLine, rendererClasses [&NiagaraSpriteRendererProperties]}
  {emitterName=Spline, rendererClasses [&NiagaraMeshRendererProperties, &NiagaraSpriteRendererProperties]}
  {emitterName=Flare001, rendererClasses []}
 ]
 userVariables [
  User.FadeOut {type=NiagaraFloat, defaultValue=1, description=""}
  User.IsFrontEnd {type=NiagaraBool, defaultValue=true}
  User.PrimaryColor {type=LinearColor, description="", defaultValue {struct=LinearColor, value=<rgba 5 0.2882 1.0735 1>}}
 ]
}
