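# Niagara telegraph for a Spire slam: a ground ring effect sized by User.Radius relative to a 1200 baseline.
# GroundRing and GroundRing002 burst flat sprites and meshes, including a delayed growing layer, using curves to drive material parameters, fades and timed brightness flashes.
# GPU emitter Minimal adds rising motes pulled by vortex and curl noise.
# System GlobalAlpha fades and kills ring particles; GroundRing1 has no renderers.
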NiagaraSystem /SpireSharedAssets/VFX/Slam/NS_Spire_Slam_TellRing
dependencies {
 usedDataInterfaces [&NiagaraDataInterfaceCurve]
 usedDynamicInputs []
 usedModules [
  &SystemState
  &self:SystemSpawnScript.SystemScriptSource.SystemScriptGraph.NiagaraNodeAssignment_0.SetVariables_#1
 ]
}
emitters {
 GroundRing {
  emitterUpdateScript [
   &EmitterState(Inactive Response=::NewEnumerator1, Life Cycle Mode=::NewEnumerator1, Loop Behavior=::NewEnumerator1, UseLoopDelay=false)
   &SpawnBurst_Instantaneous(Use Spawn Probability=false, UseLoopCountLimit=false)
  ]
  particleSpawnScript [&InitializeParticle(
    Color Mode=::NewEnumerator0
    Lifetime=4.7
    Mesh Scale=&Multiply_Vector(
     A=&MakeVector(X=&Divide_Float(A=2179.0139, B=100), Y=&Divide_Float(A=2179.0139, B=100), Z=3)
     B=&MakeVector(
      X=&Divide_Float(A=$User.Radius, B=1200)
      Y=&Divide_Float(A=$User.Radius, B=1200)
      Z=1
     )
    )
    Mesh Scale Mode=::NewEnumerator4
    Sprite Size Mode=::NewEnumerator3
    Uniform Sprite Size=2179.0139
    UsePositionOffset=false
    Write Material Random=true
   )]
  particleUpdateScript [
   &ParticleState()
   &Timeline(
    Play Rate=&FloatFromCurve(FloatCurve=DataInterface() {propertyValues {Curve {keys [{interpMode=RCIM_Cubic, tangentMode=RCTM_Break, time=0, value=27.5233}, <at 0.6 1>]}}})
    Timeline End=10000
   )
   &SpriteFacingAndAlignment(Sprite Facing=<0 0 1>, WriteAlignment=true, WriteFacing=true)
   &DynamicMaterialParameters(
    Index 0 Param 1=&FloatFromCurve(FloatCurve=DataInterface() {propertyValues {Curve {keys [
         {interpMode=RCIM_Cubic, leaveTangent=-2.8732, tangentMode=RCTM_Break, time=0, value=0.56}
         {arriveTangent=0.1229, tangentMode=RCTM_User, time=0.6, value=0}
        ]}}})
    Index 0 Param 2=&MakeFloatFromVector2d(Vector 2D=$Particles.SpriteSize)
    Index 0 Param 3=&FloatFromCurve(FloatCurve=DataInterface() {propertyValues {Curve {keys [
         {arriveTangent=-87.3831, interpMode=RCIM_Cubic, leaveTangent=-87.3831, tangentMode=RCTM_User, time=0, value=10}
         {arriveTangent=-2.1514, interpMode=RCIM_Cubic, leaveTangent=-2.1514, tangentMode=RCTM_User, time=0.2335, value=1}
         {arriveTangent=6.9549, interpMode=RCIM_Cubic, tangentMode=RCTM_Break, time=0.6806, value=2}
         {arriveTangent=61.627, interpMode=RCIM_Cubic, tangentMode=RCTM_Break, time=0.8592, value=10}
         <at 0.9564 1 RCIM_Cubic>
        ]}}})
    Index 0 Param 4=$Particles.Timeline.Time
    Index 3 Param 1=100
    Index 3 Param 2=100
    Index 3 Param 3=100
    Module.Index0Float0Write=true
    Module.Index0Float1Write=true
    Module.Index0Float2Write=true
    Module.Index0Float3Write=true
    Module.Index3Float0Write=true
    Module.Index3Float1Write=true
    Module.Index3Float2Write=true
    Module.Index3Float3Write=true
    Write Parameter Index 3=::NewEnumerator1
   )
   &ScaleColor(
    Scale Alpha=&FloatFromCurve(FloatCurve=DataInterface() {propertyValues {Curve {keys [
         <at 0 0>
         <at 0.2 1>
         {interpMode=RCIM_Cubic, tangentMode=RCTM_Break, time=0.95, value=1}
         {interpMode=RCIM_Cubic, tangentMode=RCTM_Break, time=1, value=0}
        ]}}})
    Scale RGB=&VectorFromFloat(Value=&FloatFromCurve(FloatCurve=DataInterface() {propertyValues {Curve {keys [{interpMode=RCIM_Cubic, leaveTangent=1.5269, tangentMode=RCTM_Break, time=0, value=1}, <at 0.2 1>]}}}))
    ScaleA=true
    ScaleRGB=true
   )
   &ScaleColor(
    Scale Alpha=$System.GlobalAlpha
    Scale RGB=&VectorFromFloat(Value=&FloatFromCurve(
      FloatCurve=DataInterface() {propertyValues {Curve {keys [
          {interpMode=RCIM_Cubic, tangentMode=RCTM_Break, time=3.8303, value=1}
          {interpMode=RCIM_Cubic, tangentMode=RCTM_Break, time=3.9671, value=6}
          {arriveTangent=-17.3213, interpMode=RCIM_Cubic, leaveTangent=-17.3213, tangentMode=RCTM_User, time=4.05, value=3}
          <at 4.3 1>
         ]}}}
      CurveIndex=&Add_Float(A=$Emitter.LoopedAge, B=2.7168)
     ))
    ScaleA=true
    ScaleRGB=true
   ) {module=ScaleColor001}
   &KillParticles(Kill Particles=&SetBoolByFloatComparison(A=$System.GlobalAlpha, B=0.01, Comparison Type=::NewEnumerator6))
  ]
  rendererClasses [&NiagaraSpriteRendererProperties, &NiagaraMeshRendererProperties]
  renderers [
   {rendererClass=&NiagaraSpriteRendererProperties}
   {rendererClass=&NiagaraSpriteRendererProperties, rendererIndex=1}
   {rendererClass=&NiagaraMeshRendererProperties, rendererIndex=2}
   {rendererClass=&NiagaraMeshRendererProperties, rendererIndex=3}
  ]
 }
 GroundRing002 {
  emitterUpdateScript [
   &EmitterState(Inactive Response=::NewEnumerator1, Life Cycle Mode=::NewEnumerator1, Loop Behavior=::NewEnumerator1, UseLoopDelay=false)
   &SpawnBurst_Instantaneous(Use Spawn Probability=false, UseLoopCountLimit=false)
   &SpawnBurst_Instantaneous(Spawn Group=1, Use Spawn Probability=false, UseLoopCountLimit=false) {module=SpawnBurst_Instantaneous001}
   &SpawnBurst_Instantaneous(Spawn Group=2, Spawn Time=0.5, Use Spawn Probability=false, UseLoopCountLimit=false) {module=SpawnBurst_Instantaneous002}
  ]
  particleSpawnScript [
   &InitializeParticle(
    Color Mode=::NewEnumerator0
    Lifetime=&MergeHelper_FloatSwitch(0=4.7, 1=4.7, 2=1.5, 3=0, 4=0, Index=&SpawnGroup(), Num=::NewEnumerator3)
    Sprite Size Mode=::NewEnumerator3
    Uniform Sprite Size=&MergeHelper_FloatSwitch(0=1100, 1=595, 2=1200, 3=0, 4=0, Index=&SpawnGroup(), Num=::NewEnumerator3)
    UsePositionOffset=false
    Write Material Random=true
   )
   /SetVariables(Particles.VisibilityTag=&SpawnGroup()) {isSetParameters=true}
  ]
  particleUpdateScript [
   &ParticleState()
   &Timeline(
    Play Rate=&FloatFromCurve(FloatCurve=DataInterface() {propertyValues {Curve {keys [{interpMode=RCIM_Cubic, leaveTangent=-3.7041, tangentMode=RCTM_Break, time=0, value=3.9593}, <at 0.7 0.3>, <at 0.86 0>]}}})
    Timeline End=10000
   )
   /SetVariables(Particles.Scale=&Multiply_Vector(
     A=&MakeVector(
      X=&Multiply_Float(A=&MakeFloatFromVector2d(Vector 2D=$Particles.SpriteSize), B=0.01)
      Y=&Multiply_Float(A=&MakeFloatFromVector2d(Vector 2D=$Particles.SpriteSize, Channel=::NewEnumerator1), B=0.01)
      Z=1
     )
     B=&MakeVector(
      X=&Divide_Float(A=$User.Radius, B=1200)
      Y=&Divide_Float(A=$User.Radius, B=1200)
      Z=1
     )
    )) {isSetParameters=true}
   &ScaleSpriteSize(
    Scale Sprite Size Mode=::NewEnumerator0
    Uniform Scale Factor=&MergeHelper_FloatSwitch(
     0=1
     1=1
     2=&FloatFromCurve(
      FloatCurve=DataInterface() {propertyValues {Curve {keys [
          {interpMode=RCIM_Cubic, leaveTangent=0.1511, tangentMode=RCTM_Break, time=0, value=0.4}
          {arriveTangent=0.0824, tangentMode=RCTM_User, time=1.9, value=1}
         ]}}}
      CurveIndex=$Emitter.LoopedAge
     )
     3=0
     4=0
     Index=&SpawnGroup()
     Num=::NewEnumerator3
    )
   )
   &SpriteFacingAndAlignment(Sprite Facing=<0 0 1>, WriteAlignment=true, WriteFacing=true)
   &DynamicMaterialParameters(
    Index 0 Param 1=&MergeHelper_FloatSwitch(
     0=0
     1=0
     2=&FloatFromCurve(FloatCurve=DataInterface() {propertyValues {Curve {keys [
          {interpMode=RCIM_Cubic, tangentMode=RCTM_Break, time=0, value=0}
          {arriveTangent=2.8788, tangentMode=RCTM_User, time=0.55, value=0.5}
          <at 0.75 0.1>
         ]}}})
     3=0
     4=0
     Index=&SpawnGroup()
     Num=::NewEnumerator3
    )
    Index 0 Param 2=&MergeHelper_FloatSwitch(
     0=&MakeFloatFromVector2d(Vector 2D=$Particles.SpriteSize)
     1=&MakeFloatFromVector2d(Vector 2D=$Particles.SpriteSize)
     2=&Multiply_Float(A=$Particles.Timeline.Time, B=-1)
     3=0
     4=0
     Index=&SpawnGroup()
     Num=::NewEnumerator3
    )
    Index 3 Param 1=100
    Index 3 Param 2=100
    Index 3 Param 3=100
    Module.Index0Float0Write=true
    Module.Index0Float1Write=true
    Module.Index0Float2Write=true
    Module.Index0Float3Write=true
    Module.Index3Float0Write=true
    Module.Index3Float1Write=true
    Module.Index3Float2Write=true
    Module.Index3Float3Write=true
    Write Parameter Index 3=::NewEnumerator1
   )
   &ScaleColor(
    Scale Alpha=&MergeHelper_FloatSwitch(
     0=&FloatFromCurve(
      FloatCurve=DataInterface() {propertyValues {Curve {keys [
          {interpMode=RCIM_Cubic, tangentMode=RCTM_Break, time=0, value=0}
          <at 0.3 1>
          {interpMode=RCIM_Cubic, tangentMode=RCTM_Break, time=0.95, value=1}
          {interpMode=RCIM_Cubic, tangentMode=RCTM_Break, time=1, value=0}
         ]}}}
      Scale Curve=$System.GlobalAlpha
     )
     1=&FloatFromCurve(
      FloatCurve=DataInterface() {propertyValues {Curve {keys [
          {interpMode=RCIM_Cubic, tangentMode=RCTM_Break, time=0, value=0}
          <at 0.3 1>
          {interpMode=RCIM_Cubic, tangentMode=RCTM_Break, time=0.95, value=1}
          {interpMode=RCIM_Cubic, tangentMode=RCTM_Break, time=1, value=0}
         ]}}}
      Scale Curve=$System.GlobalAlpha
     )
     2=&FloatFromCurve(FloatCurve=DataInterface() {propertyValues {Curve {keys [
          {interpMode=RCIM_Cubic, leaveTangent=0.0318, tangentMode=RCTM_Break, time=0, value=0}
          {arriveTangent=0.4738, tangentMode=RCTM_User, time=0.3292, value=1}
          {interpMode=RCIM_Cubic, tangentMode=RCTM_Break, time=0.7706, value=1}
          {interpMode=RCIM_Cubic, tangentMode=RCTM_Break, time=1, value=0}
         ]}}})
     3=0
     4=0
     Index=&SpawnGroup()
     Num=::NewEnumerator3
    )
    Scale RGB=&MergeHelper_VectorSwitch(
     0=<1 1 1>
     1=<1 1 1>
     2=&VectorFromFloat(Value=&FloatFromCurve(
       FloatCurve=DataInterface() {propertyValues {Curve {keys [
           {arriveTangent=-201.6129, interpMode=RCIM_Cubic, time=1.8, value=1}
           {interpMode=RCIM_Cubic, leaveTangent=281.5515, tangentMode=RCTM_Break, time=2.15, value=10}
           {interpMode=RCIM_Cubic, tangentMode=RCTM_Break, time=2.206, value=200}
           {arriveTangent=-703.4053, interpMode=RCIM_Cubic, leaveTangent=-703.4053, tangentMode=RCTM_User, time=2.25, value=50}
           <at 2.47 3 RCIM_Cubic>
          ]}}}
       CurveIndex=&Add_Float(A=$Emitter.LoopedAge, B=1)
      ))
     3=<0 0 0>
     4=<0 0 0>
     Index=&SpawnGroup()
     Num=::NewEnumerator3
    )
    ScaleA=true
    ScaleRGB=true
   )
   &KillParticles(Kill Particles=&SetBoolByFloatComparison(A=$System.GlobalAlpha, B=0.01, Comparison Type=::NewEnumerator6))
  ]
  rendererClasses [&NiagaraMeshRendererProperties]
  renderers [
   {rendererClass=&NiagaraMeshRendererProperties, rendererIndex=0}
   {rendererClass=&NiagaraMeshRendererProperties, rendererIndex=1}
   {rendererClass=&NiagaraMeshRendererProperties, rendererIndex=2}
  ]
 }
 GroundRing1 {rendererClasses [], renderers []}
 Minimal {
  simTarget=GPUComputeSim
  emitterUpdateScript [
   &EmitterState(Life Cycle Mode=::NewEnumerator1, Loop Behavior=::NewEnumerator1, Loop Duration=3, UseLoopDelay=false)
   &SpawnRate(
    SpawnRate=&FloatFromCurve(
     FloatCurve=DataInterface() {propertyValues {Curve {keys [
         {interpMode=RCIM_Cubic, leaveTangent=0.3372, tangentMode=RCTM_Break, time=3.9, value=0}
         {arriveTangent=1.6707, tangentMode=RCTM_User, time=4, value=1}
         <at 4.1 0>
        ]}}}
     CurveIndex=$Emitter.LoopedAge
     Scale Curve=500
    )
    Use Spawn Probability=false
   ) {enabled=false, module=SpawnRate001}
   &SpawnRate(
    SpawnRate=&FloatFromCurve(
     FloatCurve=DataInterface() {propertyValues {Curve {keys [
         {interpMode=RCIM_Cubic, leaveTangent=0.3372, tangentMode=RCTM_Break, time=0.5, value=0}
         {arriveTangent=0.27, tangentMode=RCTM_User, time=2, value=1}
         <at 3.4 1>
         <at 3.5 0>
        ]}}}
     CurveIndex=$Emitter.LoopedAge
     Scale Curve=100
    )
    Use Spawn Probability=false
   )
  ]
  particleSpawnScript [
   &InitializeParticle(
    Color=<rgba 10 0 0.8333 0.5>
    Lifetime=&RandomRangeFloat(Minimum=0.5, Maximum=1)
    Sprite Size=&Multiply_Vector2D_ByFloat(Float=&RandomRangeFloat(Minimum=0.333, Maximum=1), Vector2D=<15 50>)
    Sprite Size Mode=::NewEnumerator4
    UsePositionOffset=false
    Write Material Random=true
   )
   &ShapeLocation(
    Large Radius=&Multiply_Float(A=600, B=&Divide_Float(A=$User.Radius, B=1200))
    MaskSpawnGroup=false
    Override Seed=false
    Shape Primitive=::NewEnumerator3
    UseOwnerScale=true
   )
   &AddVelocity(Velocity=&RandomRangeVector(Minimum=<0 0 253.7016>, Maximum=<1 1 1652.3246>), Velocity Speed Scale=0.2)
  ]
  particleUpdateScript [
   &ParticleState()
   &ScaleSpriteSize(
    Scale Factor=&MakeVector2D(
     X=&FloatFromCurve(FloatCurve=DataInterface() {propertyValues {Curve {keys [
          {interpMode=RCIM_Cubic, leaveTangent=11.7691, tangentMode=RCTM_Break, time=0, value=0}
          {arriveTangent=0.0629, interpMode=RCIM_Cubic, leaveTangent=0.0629, tangentMode=RCTM_User, time=0.2, value=1}
          {interpMode=RCIM_Cubic, leaveTangent=0.027, tangentMode=RCTM_Break, time=0.35, value=1}
          {arriveTangent=0.0384, tangentMode=RCTM_User, time=1, value=0}
         ]}}})
     Y=1
    )
    Scale Sprite Size Mode=::NewEnumerator2
   )
   &VortexForce(Origin Pull Amount=647.0123, UseOriginOffset=false, Vortex Force Amount=-663.1827)
   &CurlNoiseForce(Bias Noise=false, Noise Frequency=10, Noise Strength=550.2112, Pan Noise=false)
   &SolveForcesAndVelocity(Clamp Velocity=false, Limit Acceleration=false)
   &ScaleSpriteSizeBySpeed(Max Scale Factor=<1 2>, Velocity Threshold=3500)
   &FN_Sprite_Attributes(
    Color=&MultiplyLinearColor(Linear Color=$Particles.Color, Scale Factor=<0.025 0.025 0.025 1>)
    Create Color Attribute=true
    Create SpriteSize Attribute=true
    SpriteSize=&Multiply_Vector2D(A=$Particles.SpriteSize, B=<5 2>)
    Write to Module Namespace=true
   )
   &FN_Sprite_Attributes(
    Color=&MultiplyLinearColor(Linear Color=<rgba 7 0.35 0.9042 1>, Scale Factor=<0.1 0.1 0.1 0.5>)
    Create Color Attribute=true
    Create SpriteSize Attribute=true
    SpriteSize=&Multiply_Vector2D(A=$Particles.SpriteSize, B=<10 0.5>)
    Write to Module Namespace=true
   ) {enabled=false, module=FN_Sprite_Attributes001}
   &ScaleColor(
    Scale Alpha=&FloatFromCurve(FloatCurve=DataInterface() {propertyValues {Curve {
        defaultValue=1
        keys [
         {arriveTangent=9.0725, interpMode=RCIM_Cubic, leaveTangent=9.0725, tangentMode=RCTM_User, time=0, value=0}
         {arriveTangent=-2.5249, interpMode=RCIM_Cubic, leaveTangent=-2.5249, tangentMode=RCTM_User, time=0.1408, value=0.9847}
         {arriveTangent=-0.0405, interpMode=RCIM_Cubic, leaveTangent=-0.0405, tangentMode=RCTM_User, time=1, value=0}
        ]
       }}})
    ScaleA=true
    ScaleRGB=true
   )
  ]
  rendererClasses [&NiagaraSpriteRendererProperties]
  renderers [
   {rendererClass=&NiagaraSpriteRendererProperties}
   {rendererClass=&NiagaraSpriteRendererProperties, rendererIndex=1}
   {rendererClass=&NiagaraSpriteRendererProperties, rendererIndex=2}
  ]
 }
}
summary {
 systemName=NS_Spire_Slam_TellRing
 emitters [
  {emitterName=GroundRing, rendererClasses [&NiagaraSpriteRendererProperties, &NiagaraMeshRendererProperties]}
  {emitterName=GroundRing002, rendererClasses [&NiagaraMeshRendererProperties]}
  {emitterName=Minimal, simTarget=GPUComputeSim, rendererClasses [&NiagaraSpriteRendererProperties]}
  {emitterName=GroundRing1, rendererClasses []}
 ]
 userVariables [
  User.Radius {type=NiagaraFloat, defaultValue=1200, description=""}
  User.RandomArray {
   type=NiagaraDataInterfaceArrayFloat
   defaultValue {
    struct=NiagaraExt_VariableValue_DataInterface
    value {
     dataInterface=&self:NiagaraDataInterfaceArrayFloat_0
     dataInterfaceClass=&NiagaraDataInterfaceArrayFloat
    }
   }
  }
 ]
}
